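# Translucent, unlit, two-sided Niagara material for the spline lines of the Bright Flame aura cosmetic.
# Two panning, per-particle-offset noises distort a trail texture, which drives particle-colored emissive and opacity.
# Opacity adds gradient and sphere-mask shaping, view-angle fade, and depth fade except on mobile and low quality.
# World position offset combines SplineThicken width, clamped interpolation using Niagara dynamic parameters, and optional sine/cosine wobble.
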Material /SparksCosmetics/Items/Auras/Assets/NS_Aura_BrightFlame/M_Aura_BrightFlame_SplineLines
graph {
 properties {
  bAutomaticallySetUsageInEditor=false
  bUsedWithNiagaraMeshParticles=true
  bUsedWithNiagaraSprites=true
  blendMode=BLEND_Translucent
  shadingModel=MSM_Unlit
  twoSided=true
  usedShadingModels=Unlit
 }
 nodes [
  n1 = ParticleColor()
  n2 = ScalarParameter() {defaultValue=1, parameterName=Trail Tex Tiling}
  n3 = ScalarParameter() {defaultValue=1.25, parameterName=Trail Tex TilingV}
  n4 = DynamicParameter() {parameterIndex=1, defaultValue=<rgba 0.2 1 0.1 1>, paramNames [Noise_Int, WPO_Int, SwipeOffset V, Distortion Strength]}
  n5 = Add(A=Multiply(A=TextureCoordinate(), B=AppendVector(A=@n2, B=@n3)), B=AppendVector(A=Constant(), B=@n4.SwipeOffset V))
  n6 = Multiply(A=TextureCoordinate(), B=AppendVector(A=ScalarParameter() {defaultValue=1, parameterName=Noise A Tiling}, B=Constant()))
  n7 = AppendVector(A=Constant(), B=ParticleRandom())
  n8 = ScalarParameter() {defaultValue=0.5, parameterName=Noise A Speed}
  n9 = DynamicParameter() {defaultValue=<rgba -60 1 1 1>, paramNames [Width, Pan Speed Mod, Param3, Param4]}
  n10 = Panner(
   Coordinate=MaterialFunctionCall(XY=Add(A=@n6, B=@n7)) {materialFunction=&Swizzle}.YX
   Speed=AppendVector(A=Constant(), B=Multiply(A=@n8, B=@n9.Pan Speed Mod))
  ) {speedY=4}
  n11 = Multiply(A=TextureCoordinate(), B=AppendVector(A=ScalarParameter() {defaultValue=3, parameterName=Noise B Tiling}, B=Constant()))
  n12 = AppendVector(A=Constant(), B=ParticleRandom())
  n13 = ScalarParameter() {defaultValue=-0.2, parameterName=Noise B Speed}
  n14 = Panner(
   Coordinate=MaterialFunctionCall(XY=Add(A=@n11, B=@n12)) {materialFunction=&Swizzle}.YX
   Speed=AppendVector(A=Constant(), B=Multiply(A=@n13, B=@n9.Pan Speed Mod))
  ) {speedY=4}
  n15 = Multiply(
   A=Add(
    A=TextureSample(UVs=@n10) {texture=/BRCosmetics/Weapons/FORT_Melee/Pickaxe_SoapPocket/FX/T_Pickaxe_SoapPocket_Noise}.R
    B=TextureSample(UVs=@n14) {texture=/Game/Global/Textures/Default/Masks_and_Gradients/gradient_stasis}.R
   )
   B=LinearInterpolate(Alpha=@n4.Distortion Strength) {constA=0.8, constB=1.2}
  )
  n16 = TextureSampleParameter2D!(UVs=Add(A=@n5, B=Multiply(A=@n15, B=@n4.Noise_Int) {constB=0.2}), Level, Apply View MipBias) {
   group=Trail
   mipValueMode=TMVM_MipLevel
   parameterName=Trail Tex
   sortPriority=0
   texture=/BRCosmetics/Accessories/FORT_Backpacks/Backpack_F_MED_ChimeCurlCorn/FX/T_Backpack_ChimeCurlCorn_Trail
  }
  n17 = Power(Base=@n16.RGBA, Exp=ScalarParameter() {defaultValue=0.5, parameterName=TrailTex_Power})
  n18 = ChannelMaskParameter(None=@n17) {group=Trail, maskChannel=Green, parameterName=Trail Alpha Channel, defaultValue=<rgba 0 1 0 0>}
  n19 = ScalarParameter() {group=Emissive, parameterName=Emissive Contrast}
  MaterialFunctionCall(Emissive=Multiply(
    A=@n1.RGB
    B=MaterialFunctionCall(In=@n18.R, Contrast=@n19) {materialFunction=&CheapContrast}.Result
   )) {materialFunction=&MF_FX_EmissiveAdjustments}
  n20 = TextureCoordinate()
  n21 = Add(A=Time(), B=ParticleRandom())
  n22 = LinearInterpolate(Alpha=ParticleRandom()) {constA=0.7, constB=1.3}
  n23 = Multiply(
   A=Add(A=ComponentMask(None=@n20) {g=true}, B=Multiply(A=@n21, B=@n22))
   B=ScalarParameter() {defaultValue=0.5, group=WPO, parameterName=WPOSineAmount, sortPriority=10}
  )
  n24 = Multiply(A=Sine(None=@n23), B=Constant3Vector() {constant=<rgba 0 0.75 0 1>})
  n25 = Multiply(A=Cosine(None=@n23), B=Constant3Vector() {constant=<rgba 0.75 0 0 1>})
  n26 = TextureCoordinate()
  n27 = Multiply(
   A=Multiply(A=Add(A=@n24, B=@n25), B=ScalarParameter() {defaultValue=15, group=WPO, parameterName=WPOSineIntensity, sortPriority=11})
   B=OneMinus(None=ComponentMask(None=@n26) {g=true})
  )
  n28 = DynamicParameter() {desc=Parameters set by Niagara, parameterIndex=3, defaultValue=<rgba 0 1 1 1>, paramNames [X,Y,Z,Intensity]}
  n29 = MaterialFunctionCall(Float3=Subtract(
    A=MaterialFunctionCall(X=@n28.X, Y=@n28.X, Z=@n28.X) {materialFunction=&MakeFloat3}.Result
    B=ParticlePositionWS()
   )) {materialFunction=&BreakOutFloat3Components}
  n30 = ScalarParameter() {defaultValue=100, desc=Interpolate Clamp XY, group=WPO, parameterName=Interpolate XY Clamp, sortPriority=3}
  n31 = ScalarParameter() {defaultValue=2, desc=Interpolate Clamp Z, group=WPO, parameterName=Interpolate Clamp Z, sortPriority=4}
  n32 = AppendVector(
   A=Clamp(None=AppendVector(A=@n29.R, B=@n29.R), Min=Multiply(A=@n30) {constB=-1}, Max=@n30) {maxDefault=1000, minDefault=-1000}
   B=Clamp(None=@n29.B, Min=Multiply(A=@n31) {constB=-1}, Max=@n31) {minDefault=-1}
  )
  n33 = OneMinus(None=TextureCoordinate())
  n34 = Multiply(
   A=@n28.Intensity
   B=Power(
    Base=ComponentMask(None=@n33) {g=true}
    Exp=ScalarParameter() {defaultValue=5, desc=Interpolate Power, group=WPO, parameterName=Interpolate Power, sortPriority=2}
   )
  )
  n35 = DynamicParameter() {defaultValue=<rgba -60 1 1 1>, paramNames [Width, Pan Speed Mod, Param3, Param4]}
  n36 = Add(
   A=StaticSwitchParameter(True=LinearInterpolate(B=@n32, Alpha=@n34), False=Constant()) {defaultValue=true, group=WPO, parameterName=Use Interpolate WPO, sortPriority=1}
   B=MaterialFunctionCall(WidthBase=@n35.Width, WidthTip=@n35.Width) {materialFunction=&SplineThicken}.WorldPositionOffset
  )
  NamedRerouteDeclaration(None=StaticSwitchParameter(True=Add(A=@n27, B=@n36), False=@n36) {group=WPO, parameterName=Use Wobble WPO, sortPriority=9}) {name=WPO, nodeColor=<rgba 0 0 0 1>}
  Panner(Coordinate=@n5)
  n37 = Constant()
  n38 = NamedRerouteUsage()
  Reroute(None=LinearInterpolate(A=@n37, B=@n38, Alpha=ScalarParameter() {defaultValue=1, desc="0 = disable WPO", group=WPO, parameterName=Use WPO Lerp}))
  n39 = Multiply(
   A=OneMinus(None=MaterialFunctionCall() {materialFunction=&LinearGradient}.VGradient)
   B=ScalarParameter() {defaultValue=3, parameterName=Bottom Gradient Mult}
  ) {constB=3}
  n40 = Power(Base=MaterialFunctionCall() {materialFunction=&LinearGradient}.VGradient)
  n41 = MaterialFunctionCall(
   UVs=Multiply(A=TextureCoordinate(), B=Constant2Vector() {g=0.5, r=1})
   Texture Scale=AppendVector(A=LinearInterpolate(Alpha=@n40) {constA=0.6}, B=Constant() {r=1})
  ) {materialFunction=&ScaleUVsByCenter}
  n42 = Constant() {r=0.5}
  n43 = Constant() {r=0.5}
  n44 = Constant()
  n45 = Subtract(
   A=Multiply(A=Saturate(None=@n39), B=SphereMask(A=@n41.UVs, B=@n42, Radius=@n43, Hardness=@n44) {hardnessPercent=0})
   B=ScalarParameter() {group=Texture, parameterName=Alpha Subtract, sortPriority=7}
  ) {constB=0.2}
  n46 = ScalarParameter() {defaultValue=0.25, group=Texture, parameterName=Alpha Feather}
  n47 = Power(Base=Saturate(None=Divide(A=@n45, B=@n46) {constB=0.2}))
  n48 = ChannelMaskParameter(None=@n17) {group=Trail, maskChannel=Green, parameterName=Trail Alpha Channel, defaultValue=<rgba 0 1 0 0>}
  n49 = DotProduct(A=ObjectOrientation(), B=CameraVectorWS())
  n50 = ScalarParameter() {defaultValue=5, parameterName=AngleFade}
  n51 = Multiply(A=OneMinus(None=Abs(None=@n49)), B=@n50)
  n52 = LinearInterpolate(
   A=Multiply(A=@n51, B=@n51)
   B=MaterialFunctionCall("FadeSensitivity (optional)"=@n50) {materialFunction=&FadeBasedOnViewAngle}.Result
   Alpha=ScalarParameter() {parameterName=InvertAngleFade}
  )
  n53 = Multiply(
   A=Multiply(A=@n1.A, B=Multiply(A=@n47, B=@n48.R))
   B=Multiply(A=Saturate(None=@n52), B=ScalarParameter() {defaultValue=1, parameterName=OveralOpacity})
  )
  n54 = DepthFade(Opacity=@n53, FadeDistance=ScalarParameter() {defaultValue=15, group=Opacity, parameterName=DepthFade, sortPriority=0})
  ShadingPathSwitch(Default=QualitySwitch(Default=@n54, Low=@n53, High=@n54, Medium=@n54, Epic=@n54), Mobile=@n53)
  StaticSwitchParameter(True=@n37, False=@n38) {group=WPO, parameterName=DisableWPO, sortPriority=0}
 ]
}
