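# Translucent unlit Niagara material for the flame stripes of the BrightFlame aura cosmetic.
# Panning noise drives the flame mask and color, mixing a scrolling gradient texture with top, bottom and edge tints, scaled by time of day.
# Opacity adds Niagara-driven erosion and glow, view-angle fade, and depth fade except on low quality and mobile.
# Switchable WPO combines SplineThicken width, a clamped pull from Niagara parameters and per-particle sine wobble.
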Material /SparksCosmetics/Items/Auras/Assets/NS_Aura_BrightFlame/M_Aura_BrightFlame_Stripe_Spline_Dynamic
graph {
 properties {
  bUsedWithNiagaraMeshParticles=true
  bUsedWithNiagaraSprites=true
  blendMode=BLEND_Translucent
  pixelDepthOffsetMode=PDOM_Legacy
  shadingModel=MSM_Unlit
  usedShadingModels=Unlit
 }
 nodes [
  n1 = Constant() {r=1}
  n2 = Constant() {r=2}
  n3 = ScalarParameter() {defaultValue=2, parameterName=Emissive Mult}
  n4 = VectorParameter() {parameterName=TopColor, defaultValue=<rgba 0.8879 0.3005 0.9911 1>}
  n5 = VectorParameter() {parameterName=BottomColor, defaultValue=<rgba 0.402 0.9823 0.6724 1>}
  n6 = TextureCoordinate()
  n7 = TextureCoordinate()
  n8 = Multiply(
   A=MaterialFunctionCall!(In UV=@n7) {materialFunction=/Game/Effects/Niagara/MasterMaterials/Functions/Mojo_MF/MF_FX_Symmetrical_UV}.Symmetric UV Vert
   B=AppendVector(A=ScalarParameter() {defaultValue=3, parameterName=Mask1 Tile X}, B=Constant() {r=0.1})
  )
  n9 = AppendVector(A=Constant(), B=Constant())
  n10 = Constant() {r=7}
  n11 = Multiply(A=Time(), B=ScalarParameter() {defaultValue=0.5, parameterName=Time Mod})
  n12 = ScalarParameter() {defaultValue=-0.1, parameterName=Rand Opacity Speed X}
  n13 = ScalarParameter() {defaultValue=0.02, parameterName=Rand Opacity Speed Y}
  n14 = TextureObject() {samplerType=SAMPLERTYPE_LinearColor, texture=&HighContrastNoise}
  n15 = TextureSample(UVs=Panner(Coordinate=Add(A=@n8, B=@n9), Time=Multiply(A=@n11, B=@n10), Speed=AppendVector(A=@n12, B=@n13)), Tex=@n14)
  n16 = ScalarParameter() {defaultValue=6, parameterName=Mask2 Tile X}
  n17 = Constant() {r=0.1}
  n18 = Add(A=Multiply(A=@n7, B=AppendVector(A=@n16, B=@n17)), B=AppendVector(A=Constant() {r=0.5432}, B=Constant() {r=0.2543}))
  n19 = Multiply(A=@n11, B=Constant() {r=8})
  n20 = AppendVector(
   A=ScalarParameter() {defaultValue=-0.1, parameterName=Rand Opacity Speed X}
   B=ScalarParameter() {defaultValue=0.02, parameterName=Rand Opacity Speed Y}
  )
  n21 = Power(
   Base=TextureSample(UVs=Panner(Coordinate=@n18, Time=@n19, Speed=@n20), Tex=@n14) {samplerType=SAMPLERTYPE_LinearColor, texture=&HighContrastNoise}.R
   Exp=ScalarParameter() {defaultValue=1.5, parameterName=Random Opacity Pow}
  )
  n22 = ScalarParameter() {defaultValue=0.5, parameterName=Random Opacity Secondary Pow}
  n23 = Max(A=Multiply(A=@n15.G, B=@n21), B=Multiply(A=@n15.B, B=Power(Base=@n21, Exp=@n22)))
  n24 = Saturate(None=Add(A=ComponentMask(None=@n6) {g=true}, B=Multiply(A=@n23) {constB=0.05}))
  n25 = CameraVectorWS()
  n26 = Constant() {r=0.1}
  n27 = Add(A=Add(A=ComponentMask(None=@n25) {g=true, r=true}, B=Multiply(A=@n11, B=@n26)), B=@n23)
  n28 = ParticleRandom()
  n29 = LinearInterpolate(
   A=TextureSample(
    UVs=Add(A=@n27, B=@n28)
    Tex=TextureObjectParameter() {parameterName=ColorTEX, texture=/SparksCosmetics/Items/Auras/Assets/NS_Aura_Borealis/T_Aurora_Grad_Blur_Small_D}
   ).RGB
   B=LinearInterpolate(
    A=LinearInterpolate(A=@n4.RGB, B=@n5.RGB, Alpha=@n24)
    B=VectorParameter() {parameterName=EdgeColor, defaultValue=<rgba 0.162 0.4397 0.7011 1>}.RGB
    Alpha=Power(Base=@n24)
   )
  ) {constAlpha=0.8}
  n30 = Constant() {r=-2}
  n31 = LinearInterpolate(
   A=Multiply(A=Desaturation(None=@n29, Fraction=@n30)) {constB=2, desc=Optional Boost}
   B=VectorParameter() {parameterName=Primary Color, defaultValue=<rgba 0.3278 0.956 0.9734 1>}.RGB
   Alpha=Constant() {r=0.25}
  )
  MaterialFunctionCall(Emissive=Multiply(
    A=MaterialFunctionCall(Value_Morning=@n1, Value_Day=@n2, Value_Evening=@n1, Value_Night=@n2) {materialFunction=&Scalar_Time_of_Day}.Result
    B=Multiply(A=@n3, B=@n31)
   )) {materialFunction=&MF_FX_EmissiveAdjustments}
  n32 = DynamicParameter() {desc=Parameters set by Niagara, parameterIndex=3, defaultValue=<rgba 0 1 1 1>, paramNames [X,Y,Z,Intensity]}
  n33 = MaterialFunctionCall(Float3=Subtract(
    A=MaterialFunctionCall(X=@n32.X, Y=@n32.X, Z=@n32.X) {materialFunction=&MakeFloat3}.Result
    B=ParticlePositionWS()
   )) {materialFunction=&BreakOutFloat3Components}
  n34 = ScalarParameter() {defaultValue=100, desc=Interpolate Clamp XY, group=WPO, parameterName=Interpolate XY Clamp, sortPriority=3}
  n35 = ScalarParameter() {defaultValue=2, desc=Interpolate Clamp Z, group=WPO, parameterName=Interpolate Clamp Z, sortPriority=4}
  n36 = AppendVector(
   A=Clamp(None=AppendVector(A=@n33.R, B=@n33.R), Min=Multiply(A=@n34) {constB=-1}, Max=@n34) {maxDefault=1000, minDefault=-1000}
   B=Clamp(None=@n33.B, Min=Multiply(A=@n35) {constB=-1}, Max=@n35) {minDefault=-1}
  )
  n37 = OneMinus(None=TextureCoordinate())
  n38 = Multiply(
   A=@n32.Intensity
   B=Power(
    Base=ComponentMask(None=@n37) {g=true}
    Exp=ScalarParameter() {defaultValue=5, desc=Interpolate Power, group=WPO, parameterName=Interpolate Power, sortPriority=2}
   )
  )
  n39 = DynamicParameter() {defaultValue=<rgba -60 0.2 0.2 1>, paramNames [Width, Erosion, Glow Amount, Texture Offset]}
  n40 = Add(
   A=StaticSwitchParameter(True=LinearInterpolate(B=@n36, Alpha=@n38), False=Constant()) {defaultValue=true, group=WPO, parameterName=Use Interpolate WPO, sortPriority=1}
   B=MaterialFunctionCall(WidthBase=@n39.Width, WidthTip=@n39.Width) {materialFunction=&SplineThicken}.WorldPositionOffset
  )
  n41 = ComponentMask(None=TextureCoordinate()) {g=true}
  n42 = Multiply(A=Add(A=Time(), B=ParticleRandom()), B=LinearInterpolate(Alpha=ParticleRandom()) {constA=0.7, constB=1.3})
  n43 = Sine(None=Multiply(A=Add(A=@n41, B=@n42), B=ScalarParameter() {defaultValue=0.9, group=WPO, parameterName=WPOSineAmount, sortPriority=10}))
  n44 = Constant3Vector() {constant=<rgba 0.75 0.75 0 1>}
  n45 = TextureCoordinate()
  n46 = Multiply(
   A=Multiply(A=Multiply(A=@n43, B=@n44), B=ScalarParameter() {defaultValue=15, group=WPO, parameterName=WPOSineIntensity, sortPriority=11})
   B=OneMinus(None=ComponentMask(None=@n45) {g=true})
  )
  NamedRerouteDeclaration(None=StaticSwitchParameter(True=Add(A=@n40, B=@n46), False=@n40) {group=WPO, parameterName=Use Wobble WPO, sortPriority=9}) {name=WPO, nodeColor=<rgba 0 0 0 1>}
  n47 = SphereMask(A=Multiply(A=TextureCoordinate(), B=Constant2Vector() {g=0.5, r=1}), B=Constant() {r=0.5}, Radius=Constant() {r=0.5}, Hardness=Constant()) {hardnessPercent=0}
  n48 = Multiply(A=Constant() {r=0.2}, B=@n47)
  n49 = Saturate(None=Multiply(A=@n23, B=ScalarParameter() {defaultValue=2, parameterName=Overall Alpha Mult})) {desc=opacity}
  n50 = Subtract(
   A=Multiply(A=Add(A=@n49, B=@n48), B=@n47)
   B=StaticSwitchParameter(
    True=DynamicParameter() {defaultValue=<rgba -60 0.2 0.2 1>, paramNames [Width, Erosion, Glow Amount, Texture Offset]}.Erosion
    False=ScalarParameter() {defaultValue=0.2, group=Texture, parameterName=Texture Subtract, sortPriority=7}
   ) {parameterName=Use Dynamic Subtract}
  ) {constB=0.2}
  n51 = ScalarParameter() {defaultValue=3, group=Texture, parameterName=Texture Divide, sortPriority=8}
  n52 = OneMinus(None=MaterialFunctionCall() {materialFunction=&LinearGradient}.VGradient)
  n53 = ScalarParameter() {defaultValue=3, parameterName=Bottom Gradient Mult}
  n54 = Multiply(A=TextureCoordinate(), B=Constant2Vector() {g=0.5, r=1})
  n55 = MaterialFunctionCall() {materialFunction=&LinearGradient}
  n56 = AppendVector(A=LinearInterpolate(Alpha=Power(Base=@n55.VGradient)) {constA=1}, B=Constant() {r=1})
  n57 = Multiply(
   A=Saturate(None=Multiply(A=@n52, B=@n53) {constB=3})
   B=SphereMask(
    A=MaterialFunctionCall(UVs=@n54, Texture Scale=@n56) {materialFunction=&ScaleUVsByCenter}.UVs
    B=Constant() {r=0.5}
    Radius=Constant() {r=0.5}
    Hardness=Constant()
   ) {hardnessPercent=0}
  )
  n58 = StaticSwitchParameter(
   True=DynamicParameter() {defaultValue=<rgba -60 0.2 0.2 1>, paramNames [Width, Erosion, Glow Amount, Texture Offset]}.Glow Amount
   False=ScalarParameter() {defaultValue=0.01, group=Texture, parameterName=Glow Amount}
  ) {parameterName=Use Dynamic Glow}
  n59 = Add(A=Saturate(None=Divide(A=@n50, B=@n51) {constB=1}), B=Saturate(None=Multiply(A=@n57, B=@n58) {constB=0.2})) {desc=Add some subtle glow}
  n60 = DotProduct(A=ObjectOrientation(), B=CameraVectorWS())
  n61 = ScalarParameter() {defaultValue=10, parameterName=AngleFade}
  n62 = Multiply(A=OneMinus(None=Abs(None=@n60)), B=@n61)
  n63 = Multiply(A=@n62, B=@n62)
  n64 = MaterialFunctionCall("FadeSensitivity (optional)"=@n61) {materialFunction=&FadeBasedOnViewAngle}
  n65 = ScalarParameter() {parameterName=InvertAngleFade}
  n66 = Multiply(A=Saturate(None=LinearInterpolate(A=@n63, B=@n64.Result, Alpha=@n65)), B=ScalarParameter() {defaultValue=1, parameterName=OveralOpacity})
  n67 = Multiply(A=ParticleColor().A, B=Multiply(A=Multiply(A=@n59, B=@n57), B=Saturate(None=@n66)))
  n68 = DepthFade(Opacity=@n67, FadeDistance=ScalarParameter() {defaultValue=45, group=Opacity, parameterName=DepthFade, sortPriority=0})
  ShadingPathSwitch(Default=QualitySwitch(Default=@n68, Low=@n67, High=@n68, Medium=@n68, Epic=@n68), Mobile=@n67)
  StaticSwitchParameter(True=Constant(), False=NamedRerouteUsage()) {group=WPO, parameterName=DisableWPO, sortPriority=0}
 ]
}
