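# Bright Flame aura cosmetic: a Niagara system that wraps a character's skeletal mesh in flickering flames, rising wisps and dust.
# Dust and wisp sprites spawn continuously, while burst-spawned flame meshes persist on arm and back emitters, pinned to bones by particle partition.
# A shared color curve, waveforms and panning animate tint and flicker through dynamic material parameters.
# User flags for large body and front end, plus time of day, adjust size and opacity.
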NiagaraSystem /SparksCosmetics/Items/Auras/Assets/NS_Aura_BrightFlame/NS_Aura_BrightFlame
dependencies {
 usedDataInterfaces [
  &NiagaraDataInterfaceArrayFloat
  &NiagaraDataInterfaceArrayFloat2
  &NiagaraDataInterfaceArrayInt32
  &NiagaraDataInterfaceArrayFloat3
  &NiagaraDataInterfaceCurve
  &NiagaraDataInterfaceColorCurve
  &NiagaraDataInterfaceArrayColor
  &NiagaraDataInterfaceVector2DCurve
  &NiagaraDataInterfaceSkeletalMesh
 ]
 usedDynamicInputs [&MakeFloatFromVector]
 usedModules [
  &self:SystemSpawnScript.SystemScriptSource.SystemScriptGraph.NiagaraNodeAssignment_1.SetVariables_#1
  &SystemState
  &self:SystemSpawnScript.SystemScriptSource.SystemScriptGraph.NiagaraNodeAssignment_6.SetVariables_#2
  &self:SystemSpawnScript.SystemScriptSource.SystemScriptGraph.NiagaraNodeAssignment_0.SetVariables_#3
  &self:SystemSpawnScript.SystemScriptSource.SystemScriptGraph.NiagaraNodeAssignment_4.SetVariables_#4
  &self:SystemSpawnScript.SystemScriptSource.SystemScriptGraph.NiagaraNodeAssignment_5.SetVariables_#5
 ]
}
emitters {
 Dust {
  emitterUpdateScript [
   &EmitterState()
   &SpawnRate(SpawnRate=16, Use Spawn Probability=false)
  ]
  particleSpawnScript [
   &PartitionParticles(Number Of Partitions=3, Particles Per Partition=4, Partition Mode:ENiagara_PartitionMode_v2=::NewEnumerator1)
   &CameraOffset(Camera Offset Amount=-5)
   &InitializeParticle(
    Color=<rgba 1.5 1.5 1.5 0.75>
    Lifetime Max=0.8
    Lifetime Min=0.75
    Lifetime Mode=::NewEnumerator1
    Sprite Rotation Mode=::NewEnumerator2
    Sprite Size Mode=::NewEnumerator3
    Sprite UV Mode=::NewEnumerator3
    Uniform Sprite Size=&Multiply_Float(
     A=&RandomRangeFloat(Minimum=22, Maximum=30)
     B=&MakeCustomFloatFromBool(Bool=$User.IsLargeBody, True Float=1.3, False Float=1)
    )
    UsePositionOffset=false
    Write Material Random=true
   )
   /SetVariables(Particles.VisibilityTag=&MakeCustomIntFromBool(Bool=$User.IsFrontEnd, True INT=1, False INT=0)) {isSetParameters=true}
  ]
  particleUpdateScript [
   &ParticleState()
   &SkeletalMeshLocation(
    Bone / Socket Index=&SelectIntFromArray(Direct Array Index=$Particles.PartitionParticles.Partition, Int Selection Array=DataInterface() {propertyValues {IntData [1,2,7]}})
    Bone Sampling Mode=::NewEnumerator4
    "Inherit Velocity (Scale)"=-0.1
    Inherited Velocity Speed Limit=50
    MaskSpawnGroup=false
    Orientation Sampling=::NewEnumerator1
    Sampled Position Offset=&Multiply_Vector(
     A=&Lerp_Vector(
      A=&SelectVectorFromArray(
       Direct Array Index=$Particles.PartitionParticles.Partition
       Vector Selection Array=DataInterface() {propertyValues {FloatData [<-5 0 3>, <5 0 -3>, <15 0 13>]}}
      )
      B=&SelectVectorFromArray(
       Direct Array Index=$Particles.PartitionParticles.Partition
       Vector Selection Array=DataInterface() {propertyValues {FloatData [<22 0 3>, <-22 0 -3>, <15 0 -13>]}}
      )
      Alpha=$Particles.PartitionParticles.Normalized Index In Partition
     )
     B=&MakeCustomVectorFromBool(Bool=$User.IsLargeBody, True Vector=<1.2 1 1.2>, False Vector=<1 1 1>)
    )
    Skeletal Mesh=$System.Skeletal Mesh
    Velocity Sampling=::NewEnumerator0
    VelocitySpeedLimit=true
   )
   &ScaleSpriteSize(Uniform Curve Sprite Scale=DataInterface() {propertyValues {Curve {
       defaultValue=1
       keys [
        {arriveTangent=0.511, interpMode=RCIM_Cubic, leaveTangent=0.511, tangentMode=RCTM_User, time=0, value=0.75}
        {arriveTangent=0.3756, interpMode=RCIM_Cubic, leaveTangent=0.3756, tangentMode=RCTM_User, time=1, value=1}
       ]
      }}})
   &SpriteRotationRate(Rotation Rate=30)
   &DynamicMaterialParameters(
    Index 0 Param 1=&FloatFromCurve(FloatCurve=DataInterface() {propertyValues {Curve {keys [<at 0 0 RCIM_Cubic>,<at 1.0011 0.2>]}}})
    Index 0 Param 2=&FloatFromCurve(FloatCurve=DataInterface() {propertyValues {Curve {keys [
         {arriveTangent=0.0453, interpMode=RCIM_Cubic, tangentMode=RCTM_Break, time=0, value=-1}
         {arriveTangent=0.0604, interpMode=RCIM_Cubic, leaveTangent=0.0604, tangentMode=RCTM_User, time=1, value=-1}
        ]}}})
    Index 0 Param 3=&FloatFromCurve(FloatCurve=DataInterface() {propertyValues {Curve {keys [{interpMode=RCIM_Cubic, leaveTangent=4.6779, time=0.2851, value=0.4}]}}})
    Index 1 Param 1=&FloatFromCurve(
     FloatCurve=DataInterface() {propertyValues {Curve {keys [<at 0 0>,<at 0.25 1>]}}}
     Scale Curve=&RandomRangeFloat(Minimum=2, Maximum=1)
    )
    Index 1 Param 2=&FloatFromCurve(
     FloatCurve=DataInterface() {propertyValues {Curve {keys [
         {interpMode=RCIM_Cubic, leaveTangent=3.7422, tangentMode=RCTM_Break, time=0, value=0.2}
         {interpMode=RCIM_Cubic, leaveTangent=0.9066, time=1, value=2}
        ]}}}
     Scale Curve=&RandomRangeFloat(Minimum=0.5, Maximum=1)
    )
    Index 1 Param 3=&FloatFromCurve(FloatCurve=DataInterface() {propertyValues {Curve {keys [
         {interpMode=RCIM_Cubic, leaveTangent=2.8445, tangentMode=RCTM_Break, time=0, value=0.2}
         {interpMode=RCIM_Cubic, leaveTangent=63.2978, time=0.25, value=1}
        ]}}})
    Index 1 Param 4=&FloatFromCurve(
     FloatCurve=DataInterface() {propertyValues {Curve {keys [<at 0 1>,<at 1 0.8>]}}}
     Scale Curve=&RandomRangeFloat(Minimum=0.5, Maximum=1)
    )
    Index 2 Param 1=&FloatFromCurve(FloatCurve=DataInterface() {propertyValues {Curve {keys [{interpMode=RCIM_Cubic, leaveTangent=3.2309, tangentMode=RCTM_Break, time=0.1, value=0}, <at 1 1>]}}})
    Index 2 Param 2=$Particles.MaterialRandom
    Index 2 Param 3=&FloatFromCurve(FloatCurve=DataInterface() {propertyValues {Curve {keys [
         {interpMode=RCIM_Cubic, tangentMode=RCTM_Break, time=0, value=1}
         {interpMode=RCIM_Cubic, leaveTangent=11.1233, tangentMode=RCTM_Break, time=1, value=2}
        ]}}})
    Index 2 Param 4=0.75
    Module.Param0WriteEnabled=true
    Module.Param0WriteEnabled1=false
    Module.Param0WriteEnabled2=true
    Module.Param1WriteEnabled=false
    Module.Param1WriteEnabled1=true
    Module.Param1WriteEnabled2=true
    Module.Param2WriteEnabled=false
    Module.Param2WriteEnabled1=true
    Module.Param2WriteEnabled2=true
    Module.Param3WriteEnabled=true
    Module.Param3WriteEnabled1=true
    Module.Param3WriteEnabled2=true
    Write Parameter Index 1=true
    Write Parameter Index 2=true
   )
   &ScaleColor(
    Scale Alpha=&Multiply_Float(
     A=&FloatFromCurve(
      FloatCurve=DataInterface() {propertyValues {Curve {keys [
          {interpMode=RCIM_Cubic, leaveTangent=2.1138, tangentMode=RCTM_Break, time=0, value=0}
          <at 0.3 1 RCIM_Cubic>
          {arriveTangent=-0.6585, arriveTangentWeight=0.7115, interpMode=RCIM_Cubic, leaveTangent=-0.6585, tangentMode=RCTM_User, time=1, value=0}
         ]}}}
      Scale Curve=&MakeCustomFloatFromBool(Bool=$User.IsFrontEnd, True Float=0.5, False Float=1)
     )
     B=&FloatFromCurve(
      FloatCurve=DataInterface() {propertyValues {Curve {keys [
          <at 0.1 0.85 RCIM_Cubic>
          {arriveTangentWeight=0.1077, interpMode=RCIM_Cubic, leaveTangentWeight=0.0718, tangentMode=RCTM_User, time=0.4, value=1}
          {arriveTangentWeight=0.0718, interpMode=RCIM_Cubic, leaveTangentWeight=0.1077, tangentMode=RCTM_User, time=0.6, value=1}
          <at 0.9 0.85 RCIM_Cubic>
         ]}}}
      CurveIndex=Linked() {linkedVariable {name=NPC.FortniteNPC.TimeOfDay_0-1, type=NiagaraFloat}}
     )
    )
    Scale RGB=&Multiply_Vector(
     A=&MakeVectorFromLinearColorRGB(Color=&ColorFromCurve(ColorCurve=DataInterface() {propertyValues {
         AlphaCurve {keys [<at 0 1 RCIM_Cubic>]}
         BlueCurve {keys [<at 0 1>,<at 0.0936 1.1 RCIM_Cubic>,<at 1 1 RCIM_Cubic>]}
         GreenCurve {keys [<at 0 0.6702>, {arriveTangent=-0.3634, interpMode=RCIM_Cubic, leaveTangent=-0.3634, time=0.0936, value=1.1}, <at 1 0.3068 RCIM_Cubic>]}
         RedCurve {keys [<at 0 0.9205>, {arriveTangent=-0.574, interpMode=RCIM_Cubic, leaveTangent=-0.574, time=0.0936, value=1.1}, <at 1 0.3465 RCIM_Cubic>]}
        }}))
     B=&MakeVectorFromLinearColorRGB(Color=&ColorFromCurve(
       ColorCurve=$System.Color Curve
       CurveIndex=&Frac_Float(Float=&Add_Float(A=$Particles.NormalizedAge, B=&RandomRangeFloat(Minimum=0, Maximum=1)))
      ))
    )
    ScaleA=true
    ScaleRGB=true
   ) {module=ScaleColor001}
   /SetVariables(Particles.Position Offset:NiagaraPosition=&AddVectorToPosition(
     Position=$Particles.Position
     Vector=&MakeVector(
      X=0
      Y=0
      Z=&FloatFromCurve(
       FloatCurve=DataInterface() {propertyValues {Curve {
          defaultValue=1
          keys [
           {arriveTangent=2.0114, interpMode=RCIM_Cubic, leaveTangent=2.0114, tangentMode=RCTM_User, time=0, value=0}
           {interpMode=RCIM_Cubic, tangentMode=RCTM_User, time=1, value=1}
          ]
         }}}
       Scale Curve=&Multiply_Float(
        A=&RandomRangeFloat(Minimum=10, Maximum=25)
        B=&MakeCustomFloatFromBool(Bool=$User.IsLargeBody, True Float=1.25, False Float=1)
       )
      )
     )
    )) {isSetParameters=true}
   /SetVariables(Particles.Position=$Particles.Position Offset) {isSetParameters=true}
  ]
  rendererClasses [&NiagaraSpriteRendererProperties]
  renderers [
   {rendererClass=&NiagaraSpriteRendererProperties}
   {rendererClass=&NiagaraSpriteRendererProperties, rendererIndex=1}
  ]
 }
 EnergyWisps {
  emitterUpdateScript [
   &EmitterState()
   &SpawnRate(Spawn Probability=0.75, SpawnRate=12, Use Spawn Probability=true)
  ]
  particleSpawnScript [
   &PartitionParticles(Number Of Partitions=3)
   &InitializeParticle(
    Color=&RandomRangeLinearColor(Minimum=<rgba 0.4502 0.3223 1 0.5>, Maximum=<rgba 0.6794 0.6048 1 0.6038>)
    Lifetime=&RandomRangeFloat(Minimum=0.6, Maximum=0.8)
    Sprite Size=&Multiply_Vector2D_ByFloat(
     Float=&FloatFromCurve(
      FloatCurve=DataInterface() {propertyValues {Curve {keys [<at 0.1 1 RCIM_Constant>,<at 0.3 0.9 RCIM_Constant>,<at 0.33 0.9>,<at 1 0.5>]}}}
      CurveIndex=$Particles.PartitionParticles.NormalizedPartition
     )
     Vector2D=&RandomRangeVector2D(Minimum=<10 25>, Maximum=<15 55>)
    )
    Sprite Size Mode=::NewEnumerator4
    UsePositionOffset=false
    Write Material Random=true
   )
   /SetVariables(Particles.SpriteAlignment=<0 0 1>) {isSetParameters=true}
   &AddVelocity(
    Velocity=<0 0 1>
    Velocity Speed Scale=&Multiply_Float(
     A=&Multiply_Float(
      A=&RandomRangeFloat(Minimum=225, Maximum=300)
      B=&Multiply_Float(
       A=&FloatFromCurve(
        FloatCurve=DataInterface() {propertyValues {Curve {keys [<at 0.3 1>,<at 1 0.75>]}}}
        CurveIndex=$Particles.PartitionParticles.NormalizedPartition
       )
       B=1
      )
     )
     B=&MakeCustomFloatFromBool(Bool=$User.IsLargeBody, True Float=1.2, False Float=1)
    )
   )
   &CameraOffset(Camera Offset Amount=-15)
  ]
  particleUpdateScript [
   &ParticleState()
   &SkeletalMeshLocation(
    Bone / Socket Index=&SelectIntFromArray(Direct Array Index=$Particles.PartitionParticles.Partition, Int Selection Array=DataInterface() {propertyValues {IntData [7,1,2,3,4]}})
    Bone Sampling Mode=::NewEnumerator4
    "Inherit Velocity (Scale)"=-0.05
    Inherited Velocity Speed Limit=25
    MaskSpawnGroup=false
    Orientation Sampling=::NewEnumerator1
    Position Sampling=::NewEnumerator1
    Sampled Position Offset=&Multiply_Vector(A=&RandomRangeVector(Minimum=<-1 -1 -1>, Maximum=<1 1 1>), B=<10 10 10>)
    Skeletal Mesh=$System.Skeletal Mesh
    Velocity Sampling=::NewEnumerator0
    VelocitySpeedLimit=true
   )
   /SetVariables(
    Particles.Pan:NiagaraFloat=&FloatFromCurve(
     FloatCurve=DataInterface() {propertyValues {Curve {
        defaultValue=0
        keys [<at 0.0048 0.0238>,<at 0.1378 0.0717>,<at 0.3024 0.0558>,<at 0.4966 0.0912>,<at 0.7069 0.0356>,<at 0.8276 0.0931>,<at 1.0048 0.0238>]
       }}}
     CurveIndex=&Frac_Float(Float=&Add_Float(
       A=&Multiply_Float(A=$Engine.Time, B=&RandomRangeFloat(Minimum=0.75, Maximum=2))
       B=&RandomRangeFloat(Minimum=0, Maximum=1)
      ))
     Scale Curve=0.8
    )
    Particles.Waveform:NiagaraFloat=&Waveform(
     Global Phase=&Add_Float(A=$Particles.Initial.PartitionParticles.NormalizedPartition, B=$Engine.Time)
     "Waveform [1]"=::NewEnumerator12
     "[1] Frequency"=3
     "[1] Amplitude Min/Max"=<0.75 1.1>
    )
   ) {isSetParameters=true}
   &DynamicMaterialParameters(
    Index 1 Param 1=0.1
    Index 1 Param 3=&FloatFromCurve(FloatCurve=DataInterface() {propertyValues {Curve {keys [<at 0 0 RCIM_Cubic>, {arriveTangent=1.7958, interpMode=RCIM_Cubic, leaveTangent=1.7958, tangentMode=RCTM_User, time=1, value=0.5}]}}})
    Module.Index0Float0Write=true
    Module.Index0Float1Write=true
    Module.Index0Float2Write=true
    Module.Index0Float3Write=true
    Module.Index1Float0Write=true
    Module.Index1Float1Write=true
    Module.Index1Float2Write=true
    Module.Index1Float3Write=true
    Write Parameter Index 1=::NewEnumerator1
   )
   &ScaleSpriteSize(
    Scale Factor=&MakeVector2D(
     X=&Multiply_Float(A=1, B=$Particles.Waveform)
     Y=&FloatFromCurve(FloatCurve=DataInterface() {propertyValues {Curve {
         defaultValue=1
         keys [
          {interpMode=RCIM_Cubic, leaveTangentWeight=0.3747, tangentMode=RCTM_User, time=0, value=1}
          {arriveTangent=-1.6085, interpMode=RCIM_Cubic, leaveTangent=-1.6085, tangentMode=RCTM_User, time=1, value=0.5}
         ]
        }}})
    )
    Scale Sprite Size Mode=::NewEnumerator2
   )
   &Drag(Drag=5.25, Use Linear Drag=true, Use Rotational Drag=true)
   &SolveForcesAndVelocity(Clamp Velocity=false, Limit Acceleration=false)
   &ScaleColor(
    Scale Alpha=&FloatFromCurve(
     FloatCurve=DataInterface() {propertyValues {Curve {
        defaultValue=1
        keys [
         {arriveTangent=8.4968, interpMode=RCIM_Cubic, leaveTangent=8.4968, tangentMode=RCTM_User, time=0, value=0.5}
         {interpMode=RCIM_Cubic, tangentMode=RCTM_User, time=0.2, value=1}
         {arriveTangent=-0.0182, interpMode=RCIM_Cubic, leaveTangent=-0.0182, tangentMode=RCTM_User, time=1, value=0}
        ]
       }}}
     Scale Curve=&FloatFromCurve(
      FloatCurve=DataInterface() {propertyValues {Curve {keys [
          <at 0.1 0.85 RCIM_Cubic>
          {arriveTangentWeight=0.1077, interpMode=RCIM_Cubic, leaveTangentWeight=0.0718, tangentMode=RCTM_User, time=0.4, value=1}
          {arriveTangentWeight=0.0718, interpMode=RCIM_Cubic, leaveTangentWeight=0.1077, tangentMode=RCTM_User, time=0.6, value=1}
          <at 0.9 0.85 RCIM_Cubic>
         ]}}}
      CurveIndex=Linked() {linkedVariable {name=NPC.FortniteNPC.TimeOfDay_0-1, type=NiagaraFloat}}
      Scale Curve=&MakeCustomFloatFromBool(Bool=$User.IsFrontEnd, True Float=0.4, False Float=1)
     )
    )
    Scale RGB=&MakeVectorFromLinearColorRGB(Color=&ColorFromCurve(
      ColorCurve=$System.Color Curve
      CurveIndex=&Frac_Float(Float=&Add_Float(
        A=&Multiply_Float(A=$Engine.Time, B=0.5)
        B=&RandomRangeFloat(Minimum=0, Maximum=1)
       ))
     ))
    ScaleA=true
    ScaleRGB=true
   )
  ]
  rendererClasses [&NiagaraSpriteRendererProperties]
  renderers [{rendererClass=&NiagaraSpriteRendererProperties}]
 }
 Flames_Arms {
  emitterUpdateScript [
   &EmitterState(Life Cycle Mode=::NewEnumerator1, Loop Behavior=::NewEnumerator1, Loop Duration Mode=::NewEnumerator1, UseLoopDelay=false)
   &SpawnBurst_Instantaneous(Spawn Count=16, Use Spawn Probability=false, UseLoopCountLimit=false)
  ]
  particleSpawnScript [
   &PartitionParticles(Partition Mode:ENiagara_PartitionMode_v2=::NewEnumerator1) {module=PartitionParticles001}
   /SetVariables(Particles.VisibilityTag=$Particles.PartitionParticles001.Partition) {isSetParameters=true}
   &PartitionParticles(Normalized Index In Partition Includes One=true, Particles Per Partition=4, Partition Mode:ENiagara_PartitionMode_v2=::NewEnumerator1)
   &InitializeParticle(
    Color=&SelectLinearColorFromArray(
     Array Sampling Mode=::NewEnumerator1
     Direct Array Index=$Particles.PartitionParticles001.Partition
     Color Selection Array=DataInterface() {propertyValues {ColorData [<rgba 0.9298 0.8099 1 0.75>, <rgba 0.2859 0.1198 1 0.65>]}}
     Random Seed=?
    )
    Lifetime=0.1
    Mesh Scale=&MakeVector(
     X=1
     Y=1
     Z=&Multiply_Float(
      A=&Lerp_Float(A=1.05, B=0.6, Alpha=$Particles.PartitionParticles.Normalized Index In Partition)
      B=&Multiply_Float(
       A=&MakeCustomFloatFromBool(Bool=$User.IsLargeBody, True Float=1.25, False Float=1)
       B=$System.NewFlameHeight
      )
     )
    )
    Mesh Scale Mode=::NewEnumerator4
    UsePositionOffset=false
    Write Material Random=true
   )
   &InitialMeshOrientation(Mesh Orientation Mode=::NewEnumerator1, Use Rotation Vector=true)
  ]
  particleUpdateScript [
   &ParticleState(Kill Particles When Lifetime Has Elapsed=false, Let Infinitely Lived Particles Die When Emitter Deactivates=true)
   /SetVariables(
    Particles.Pan:NiagaraFloat=&FloatFromCurve(
     FloatCurve=DataInterface() {propertyValues {Curve {defaultValue=0, keys [<at 0 0.1626>,<at 0.1456 0.2105>,<at 0.306 0.15>,<at 0.5044 0.25>,<at 0.7147 0.1745>,<at 0.8353 0.26>,<at 1 0.1626>]}}}
     CurveIndex=&Frac_Float(Float=&Add_Float(
       A=&Multiply_Float(A=$Engine.Time, B=&RandomRangeFloat(Minimum=1, Maximum=3))
       B=&RandomRangeFloat(Minimum=0, Maximum=1)
      ))
     Scale Curve=0.75
    )
    Particles.Waveform:NiagaraFloat=&Waveform(
     Global Phase=&Add_Float(A=$Particles.PartitionParticles.Normalized Index In Partition, B=$Engine.Time)
     "Waveform [1]"=::NewEnumerator12
     "[1] Frequency"=3
     "[1] Amplitude Min/Max"=<0.75 1.1>
    )
   ) {isSetParameters=true}
   &DynamicMaterialParameters(
    Index 0 Param 1=&Multiply_Float(
     A=&Multiply_Float(
      A=&Lerp_Float(A=19, B=17, Alpha=$Particles.PartitionParticles.Normalized Index In Partition)
      B=&Multiply_Float(
       A=$Particles.Waveform
       B=&MakeCustomFloatFromBool(Bool=$User.IsLargeBody, True Float=1.3, False Float=1)
      )
     )
     B=&SelectFloatFromArray(
      Direct Array Index=$Particles.PartitionParticles001.Partition
      Float Selection Array=DataInterface() {propertyValues {FloatData [0.85,1.15]}}
     )
    )
    Index 0 Param 2=&RandomRangeFloat(Minimum=0.5, Maximum=0.7)
    Index 1 Param 1=0.1
    Index 1 Param 3=$Particles.Pan
    Module.Index0Param0Write=true
    Module.Index0Param1Write=true
    Module.Index0Param2Write=true
    Module.Index0Param3Write=true
    Module.Index1Param0Write=true
    Module.Index1Param1Write=true
    Module.Index1Param2Write=true
    Module.Index1Param3Write=true
    Write Parameter Index 1=true
   )
   &UpdateMeshOrientation(
    Facing Direction=/Game/Effects/Niagara/DynamicInputs/Engine/FN_GetSystemAxis(Axis:ENiagaraOrientationAxis=::XAxis, Flip Axis:NiagaraBool=false)
    Facing Interpolation=false
    Orientation Method=::NewEnumerator1
    Override Quat=false
    TransformFacing=false
    TransformUp=false
   )
   &SkeletalMeshLocation(
    Bone / Socket Index=&SelectIntFromArray(Direct Array Index=$Particles.PartitionParticles.Partition, Int Selection Array=DataInterface() {propertyValues {IntData [1,2]}})
    Bone Sampling Mode=::NewEnumerator4
    MaskSpawnGroup=false
    Orientation Sampling=::NewEnumerator1
    Sampled Position Offset=&Multiply_Vector(
     A=&Lerp_Vector(
      A=&SelectVectorFromArray(
       Direct Array Index=$Particles.PartitionParticles.Partition
       Vector Selection Array=DataInterface() {propertyValues {FloatData [<5 -2 -1>, <-5 2 1>]}}
      )
      B=&SelectVectorFromArray(
       Direct Array Index=$Particles.PartitionParticles.Partition
       Vector Selection Array=DataInterface() {propertyValues {FloatData [<19 0 -2>, <-19 0 2>]}}
      )
      Alpha=$Particles.PartitionParticles.Normalized Index In Partition
     )
     B=&Multiply_VectorByFloat(
      Vector=&MakeCustomVectorFromBool(Bool=$User.IsLargeBody, True Vector=<1.05 1 0.5>, False Vector=<1 1 1>)
      Float=$System.EngineOwnerScale
     )
    )
    Skeletal Mesh=$System.Skeletal Mesh
   ) {module=SkeletalMeshLocation001}
   &InterpolateOverTime(
    Rate Of Change=12
    Target Vector Value:Vector3f=&ConvertPositionToVector(Input Position=$Particles.Position, Convert To Absolute Large World Vector=false)
    Variable Type=::NewEnumerator2
   ) {enabled=false}
   /SetVariables(Particles.Interpolate Dynamic=&MakeFloat4FromVectorAndFloat("Vector (RGB)"=<1 1 1>, "Float (Alpha)"=0)) {isSetParameters=true, module=SetVariables_#6}
   &ScaleColor(
    Scale Alpha=&MaskFloatBySpawnGroup(
     Masked Float=&Lerp_Float(A=0.5, B=0.5, Alpha=$Particles.PartitionParticles.Normalized Index In Partition)
     Passthrough Float=1
     Spawn Group Mask=&MakeIntFromFloat(Float=&Lerp_Float(A=0.5, B=0.25, Alpha=$Particles.PartitionParticles.Normalized Index In Partition))
    )
    Scale RGB=$System.Color RGB
    ScaleA=true
    ScaleRGB=true
   ) {enabled=false}
   &ScaleColor(
    Scale Alpha=&Multiply_Float(
     A=&Lerp_Float(A=1, B=0.5, Alpha=$Particles.PartitionParticles.Normalized Index In Partition)
     B=&Multiply_Float(
      A=&Waveform(
       Global Phase=&Add_Float(A=$Particles.PartitionParticles.Normalized Index In Partition, B=$Engine.Time)
       "Waveform [1]"=::NewEnumerator9
       "[1] Frequency"=3
       "[1] Amplitude Min/Max"=<0.4 1>
      )
      B=&FloatFromCurve(
       FloatCurve=DataInterface() {propertyValues {Curve {keys [
           <at 0.1 0.85 RCIM_Cubic>
           {arriveTangentWeight=0.1077, interpMode=RCIM_Cubic, leaveTangentWeight=0.0718, tangentMode=RCTM_User, time=0.4, value=1}
           {arriveTangentWeight=0.0718, interpMode=RCIM_Cubic, leaveTangentWeight=0.1077, tangentMode=RCTM_User, time=0.6, value=1}
           <at 0.9 0.85 RCIM_Cubic>
          ]}}}
       CurveIndex=Linked() {linkedVariable {name=NPC.FortniteNPC.TimeOfDay_0-1, type=NiagaraFloat}}
       Scale Curve=&MakeCustomFloatFromBool(
        Bool=$User.IsFrontEnd
        True Float=&SelectFloatFromArray(Direct Array Index=$Particles.PartitionParticles001.Partition, Float Selection Array=DataInterface() {propertyValues {FloatData [0.6,1]}})
        False Float=1
       )
      )
     )
    )
    Scale RGB=&MakeVectorFromLinearColorRGB(Color=&ColorFromCurve(
      ColorCurve=$System.Color Curve
      CurveIndex=&Frac_Float(Float=&Add_Float(A=&Multiply_Float(A=$Engine.Time, B=0.3), B=&ReturnNormalizedExecIndex()))
     ))
    ScaleA=true
    ScaleRGB=true
   ) {module=ScaleColor001}
  ]
  rendererClasses [&NiagaraMeshRendererProperties]
  renderers [
   {rendererClass=&NiagaraMeshRendererProperties, rendererIndex=0}
   {rendererClass=&NiagaraMeshRendererProperties, rendererIndex=1}
   {rendererClass=&NiagaraMeshRendererProperties, rendererIndex=2}
  ]
 }
 Flames_Back_Original {
  bEnabled=false
  emitterUpdateScript [
   &EmitterState(Life Cycle Mode=::NewEnumerator1, Loop Behavior=::NewEnumerator1, Loop Duration Mode=::NewEnumerator1, UseLoopDelay=false)
   &SpawnBurst_Instantaneous(Spawn Count=12, Use Spawn Probability=false, UseLoopCountLimit=false)
  ]
  particleSpawnScript [
   &PartitionParticles(Normalized Index In Partition Includes One=true, Particles Per Partition=6, Partition Mode:ENiagara_PartitionMode_v2=::NewEnumerator1) {module=PartitionParticles001}
   /SetVariables(Particles.VisibilityTag=$Particles.PartitionParticles001.Partition) {isSetParameters=true}
   &InitializeParticle(
    Color=&SelectLinearColorFromArray(
     Array Sampling Mode=::NewEnumerator1
     Direct Array Index=$Particles.PartitionParticles001.Partition
     Color Selection Array=DataInterface() {propertyValues {ColorData [<rgba 0.9298 0.8099 1 0.9>, <rgba 0.2859 0.1198 1 0.9>]}}
     Random Seed=?
    )
    Lifetime=0.1
    Mesh Scale=&MakeVector(
     X=1
     Y=1
     Z=&FloatFromCurve(
      FloatCurve=DataInterface() {propertyValues {Curve {
         defaultValue=1
         keys [
          {interpMode=RCIM_Cubic, leaveTangentWeight=0.1667, tangentMode=RCTM_User, time=0, value=1.2}
          {arriveTangentWeight=0.1667, interpMode=RCIM_Cubic, leaveTangentWeight=0.1667, tangentMode=RCTM_User, time=0.5, value=2.2}
          {tangentMode=RCTM_User, time=1, value=1.2}
         ]
        }}}
      CurveIndex=$Particles.PartitionParticles001.Normalized Index In Partition
      Scale Curve=&Multiply_Float(
       A=&MakeCustomFloatFromBool(Bool=$User.IsLargeBody, True Float=1.25, False Float=1)
       B=$System.NewFlameHeight
      )
     )
    )
    Mesh Scale Mode=::NewEnumerator4
    UsePositionOffset=false
    Write Material Random=true
   )
   &InitialMeshOrientation(Mesh Orientation Mode=::NewEnumerator1, Use Rotation Vector=true)
  ]
  particleUpdateScript [
   /SetVariables(
    Particles.Pan:NiagaraFloat=&FloatFromCurve(
     FloatCurve=DataInterface() {propertyValues {Curve {
        defaultValue=0
        keys [<at 0 0.0425>,<at 0.1304 0.0904>,<at 0.2933 0.0478>,<at 0.4901 0.119>,<at 0.579 0.0524>,<at 0.8202 0.1118>,<at 1 0.0425>]
       }}}
     CurveIndex=&Frac_Float(Float=&Add_Float(
       A=&Multiply_Float(A=$Engine.Time, B=&RandomRangeFloat(Minimum=0.75, Maximum=2))
       B=&RandomRangeFloat(Minimum=0, Maximum=1)
      ))
     Scale Curve=0.75
    )
    Particles.Waveform:NiagaraFloat=&Waveform(
     Global Phase=&Add_Float(A=$Particles.PartitionParticles001.NormalizedPartition, B=$Engine.Time)
     "Waveform [1]"=::NewEnumerator12
     "[1] Frequency"=3
     "[1] Amplitude Min/Max"=<0.75 1.1>
    )
   ) {isSetParameters=true, module=SetVariables_#7}
   &ParticleState(Kill Particles When Lifetime Has Elapsed=false, Let Infinitely Lived Particles Die When Emitter Deactivates=true)
   &DynamicMaterialParameters(
    Index 0 Param 1=&Multiply_Float(
     A=&Multiply_Float(
      A=&FloatFromCurve(
       FloatCurve=DataInterface() {propertyValues {Curve {
          defaultValue=1
          keys [
           {interpMode=RCIM_Cubic, tangentMode=RCTM_User, time=0, value=0.85}
           {interpMode=RCIM_Cubic, tangentMode=RCTM_User, time=0.5, value=1}
           {tangentMode=RCTM_User, time=1, value=0.85}
          ]
         }}}
       CurveIndex=$Particles.PartitionParticles001.Normalized Index In Partition
       Scale Curve=&Multiply_Float(A=24, B=$Particles.Waveform)
      )
      B=&MakeCustomFloatFromBool(Bool=$User.IsLargeBody, True Float=1.4, False Float=1)
     )
     B=&SelectFloatFromArray(
      Direct Array Index=$Particles.PartitionParticles001.Partition
      Float Selection Array=DataInterface() {propertyValues {FloatData [0.85,1.15]}}
     )
    )
    Index 0 Param 2=&RandomRangeFloat(Minimum=0.5, Maximum=1.2)
    Index 1 Param 1=0.13
    Index 1 Param 3=$Particles.Pan
    Index 1 Param 4=&Waveform(
     Waveform Count=::NewEnumerator1
     Global Phase=&Add_Float(A=$Particles.PartitionParticles001.NormalizedPartition, B=$Engine.Time)
     "Waveform [1]"=::NewEnumerator13
     "[1] Frequency"=1
     "[1] Amplitude Min/Max"=<0 1>
     "[2] Average"=false
     "Waveform [2]"=::NewEnumerator9
     "[2] Input Position"=0
     "[2] Phase"=$Emitter.Age
     "[2] Phase Scale"=1
     "[2] Compound Frequency"=2
     "[2] Frequency"=2
     "[2] Amplitude Min/Max"=<0 1>
     "[2] Scale"=1
     "[2] Bias"=0
     "[2] Exponent"=1
     "[2] Pulse Duration"=0.1
     "[2] Clip Triangle Min/Max"=<0 0>
     "[2] Random Seed"=0
    )
    Module.Index0Param0Write=true
    Module.Index0Param1Write=true
    Module.Index0Param2Write=true
    Module.Index0Param3Write=true
    Module.Index1Param0Write=true
    Module.Index1Param1Write=true
    Module.Index1Param2Write=true
    Module.Index1Param3Write=true
    Write Parameter Index 1=true
   )
   &UpdateMeshOrientation(
    Facing Direction=&FN_GetSystemAxis(Axis:ENiagaraOrientationAxis=::YAxis, Flip Axis:NiagaraBool=false)
    Facing Interpolation=false
    Orientation Method=::NewEnumerator1
    Override Quat=false
    TransformFacing=false
    TransformUp=false
   )
   &SkeletalMeshLocation(
    Bone / Socket Index=7
    Bone Sampling Mode=::NewEnumerator4
    MaskSpawnGroup=false
    Orientation Sampling=::NewEnumerator1
    Sampled Position Offset=&Multiply_Vector(
     A=&MakeVector(
      X=&FloatFromCurve(
       FloatCurve=DataInterface() {propertyValues {Curve {
          defaultValue=1
          keys [
           {interpMode=RCIM_Cubic, tangentMode=RCTM_User, time=0, value=5}
           {interpMode=RCIM_Cubic, tangentMode=RCTM_User, time=0.5, value=2}
           {tangentMode=RCTM_User, time=1, value=5}
          ]
         }}}
       CurveIndex=$Particles.PartitionParticles001.Normalized Index In Partition
      )
      Y=&FloatFromCurve(
       FloatCurve=DataInterface() {propertyValues {Curve {keys [
           {interpMode=RCIM_Cubic, leaveTangentWeight=0.1667, tangentMode=RCTM_User, time=0, value=0}
           {arriveTangentWeight=0.1667, interpMode=RCIM_Cubic, leaveTangentWeight=0.1667, tangentMode=RCTM_User, time=0.5, value=-3}
           {arriveTangentWeight=0.1667, interpMode=RCIM_Cubic, tangentMode=RCTM_User, time=1, value=0}
          ]}}}
       CurveIndex=$Particles.PartitionParticles001.Normalized Index In Partition
      )
      Z=&FloatFromCurve(
       FloatCurve=DataInterface() {propertyValues {Curve {keys [<at 0 14>,<at 1 -14>]}}}
       CurveIndex=$Particles.PartitionParticles001.Normalized Index In Partition
      )
     )
     B=&Multiply_VectorByFloat(
      Vector=&MakeCustomVectorFromBool(Bool=$User.IsLargeBody, True Vector=<1 1 1.18>, False Vector=<1 1 1>)
      Float=$System.EngineOwnerScale
     )
    )
    Skeletal Mesh=$System.Skeletal Mesh
   )
   &SkeletalMeshLocation(
    Bone / Socket Index=&SelectIntFromArray(Direct Array Index=$Particles.PartitionParticles.Partition, Int Selection Array=DataInterface() {propertyValues {IntData [1,2]}})
    Bone Sampling Mode=::NewEnumerator4
    MaskSpawnGroup=false
    Orientation Sampling=::NewEnumerator1
    Sampled Position Offset=&Lerp_Vector(
     A=&SelectVectorFromArray(Direct Array Index=$Particles.PartitionParticles.Partition, Vector Selection Array=DataInterface() {propertyValues {FloatData []}})
     B=&SelectVectorFromArray(
      Direct Array Index=$Particles.PartitionParticles.Partition
      Vector Selection Array=DataInterface() {propertyValues {FloatData [<25 0 0>, <-25 0 0>]}}
     )
     Alpha=$Particles.PartitionParticles.Normalized Index In Partition
    )
    Skeletal Mesh=$System.Skeletal Mesh
   ) {enabled=false, module=SkeletalMeshLocation001}
   &InterpolateOverTime(
    Rate Of Change=12
    Target Vector Value:Vector3f=&ConvertPositionToVector(Input Position=$Particles.Position, Convert To Absolute Large World Vector=false)
    Variable Type=::NewEnumerator2
   )
   /SetVariables(Particles.Interpolate Dynamic=&MakeFloat4FromVectorAndFloat("Vector (RGB)"=$StackContext.InterpolateOverTime.InterpolateOverTime_V3_.Moving Average, "Float (Alpha)"=0)) {isSetParameters=true, module=SetVariables_#6}
   &ScaleColor(Scale Alpha=0.5, Scale RGB=$System.Color RGB, ScaleA=true, ScaleRGB=true) {enabled=false}
   &ScaleColor(
    Scale Alpha=&Multiply_Float(
     A=&Waveform(
      Global Phase=&Add_Float(A=&ReturnNormalizedExecIndex(), B=$Engine.Time)
      "Waveform [1]"=::NewEnumerator9
      "[1] Frequency"=5
      "[1] Amplitude Min/Max"=<0.4 1>
     )
     B=&FloatFromCurve(
      FloatCurve=DataInterface() {propertyValues {Curve {keys [
          <at 0.1 0.85 RCIM_Cubic>
          {arriveTangentWeight=0.1077, interpMode=RCIM_Cubic, leaveTangentWeight=0.0718, tangentMode=RCTM_User, time=0.4, value=1}
          {arriveTangentWeight=0.0718, interpMode=RCIM_Cubic, leaveTangentWeight=0.1077, tangentMode=RCTM_User, time=0.6, value=1}
          <at 0.9 0.85 RCIM_Cubic>
         ]}}}
      CurveIndex=Linked() {linkedVariable {name=NPC.FortniteNPC.TimeOfDay_0-1, type=NiagaraFloat}}
      Scale Curve=&MakeCustomFloatFromBool(Bool=$User.IsFrontEnd, True Float=0.8, False Float=1)
     )
    )
    Scale RGB=&MakeVectorFromLinearColorRGB(Color=&ColorFromCurve(
      ColorCurve=$System.Color Curve
      CurveIndex=&Frac_Float(Float=&Add_Float(A=&Multiply_Float(A=$Engine.Time, B=0.3), B=$Particles.PartitionParticles001.Normalized Index In Partition))
     ))
    ScaleA=true
    ScaleRGB=true
   ) {module=ScaleColor001}
  ]
  rendererClasses [&NiagaraMeshRendererProperties]
  renderers [
   {rendererClass=&NiagaraMeshRendererProperties, rendererIndex=0}
   {rendererClass=&NiagaraMeshRendererProperties, rendererIndex=1}
   {rendererClass=&NiagaraMeshRendererProperties, rendererIndex=2}
  ]
 }
 Flames_Back_Revised {
  emitterUpdateScript [
   &EmitterState()
   &SpawnBurst_Instantaneous(Spawn Count=6, Use Spawn Probability=false, UseLoopCountLimit=true)
  ]
  particleSpawnScript [
   &PartitionParticles(Normalized Index In Partition Includes One=true, Particles Per Partition=3, Partition Mode:ENiagara_PartitionMode_v2=::NewEnumerator1) {module=PartitionParticles001}
   &InitializeParticle(
    Color=<rgba 0.9298 0.8099 1 0.9>
    Lifetime=0.1
    Mesh Scale=&MakeVector(
     X=1
     Y=1
     Z=&FloatFromCurve(
      FloatCurve=DataInterface() {propertyValues {Curve {
         defaultValue=1
         keys [
          <at 0 1.2 RCIM_Cubic>
          {interpMode=RCIM_Cubic, leaveTangentWeight=0.1667, tangentMode=RCTM_User, time=0.5, value=1.8}
          {arriveTangentWeight=0.1667, interpMode=RCIM_Cubic, leaveTangentWeight=0.1667, tangentMode=RCTM_User, time=1, value=1.2}
         ]
        }}}
      CurveIndex=&ReturnNormalizedExecIndex()
      Scale Curve=&Multiply_Float(
       A=&MakeCustomFloatFromBool(Bool=$User.IsLargeBody, True Float=1.25, False Float=1)
       B=$System.NewFlameHeight
      )
     )
    )
    Mesh Scale Mode=::NewEnumerator4
    UsePositionOffset=false
    Write Material Random=true
   )
   &InitialMeshOrientation(Mesh Orientation Mode=::NewEnumerator1, Use Rotation Vector=true)
   /SetVariables(Particles.BackerFlameColor:LinearColor=<rgba 0.2859 0.1198 1 0.9>) {isSetParameters=true}
  ]
  particleUpdateScript [
   /SetVariables(
    Particles.Pan:NiagaraFloat=&FloatFromCurve(
     FloatCurve=DataInterface() {propertyValues {Curve {
        defaultValue=0
        keys [<at 0 0.0425>,<at 0.1304 0.0904>,<at 0.2933 0.0478>,<at 0.4901 0.119>,<at 0.579 0.0524>,<at 0.8202 0.1118>,<at 1 0.0425>]
       }}}
     CurveIndex=&Frac_Float(Float=&Add_Float(
       A=&Multiply_Float(A=$Engine.Time, B=&RandomRangeFloat(Minimum=0.75, Maximum=2))
       B=&RandomRangeFloat(Minimum=0, Maximum=1)
      ))
     Scale Curve=0.75
    )
    Particles.Waveform:NiagaraFloat=&Waveform(
     Global Phase=&Add_Float(A=&ReturnNormalizedExecIndex(), B=$Engine.Time)
     "Waveform [1]"=::NewEnumerator12
     "[1] Frequency"=3
     "[1] Amplitude Min/Max"=<0.75 1.1>
    )
   ) {isSetParameters=true, module=SetVariables_#7}
   &ParticleState(Kill Particles When Lifetime Has Elapsed=false, Let Infinitely Lived Particles Die When Emitter Deactivates=true)
   &DynamicMaterialParameters(
    Index 0 Param 1=&Multiply_Float(
     A=&Multiply_Float(
      A=&FloatFromCurve(
       FloatCurve=DataInterface() {propertyValues {Curve {
          defaultValue=1
          keys [
           {interpMode=RCIM_Cubic, tangentMode=RCTM_User, time=0, value=0.85}
           {interpMode=RCIM_Cubic, tangentMode=RCTM_User, time=0.5, value=1}
           <at 1 0.85 RCIM_Cubic>
          ]
         }}}
       CurveIndex=&ReturnNormalizedExecIndex()
       Scale Curve=&Multiply_Float(A=25, B=$Particles.Waveform)
      )
      B=&MakeCustomFloatFromBool(Bool=$User.IsLargeBody, True Float=1.4, False Float=1)
     )
     B=0.85
    )
    Index 0 Param 2=&RandomRangeFloat(Minimum=0.5, Maximum=1.2)
    Index 1 Param 1=0.13
    Index 1 Param 3=$Particles.Pan
    Index 1 Param 4=&Waveform(
     Waveform Count=::NewEnumerator1
     Global Phase=&Add_Float(A=$Particles.PartitionParticles001.NormalizedPartition, B=$Engine.Time)
     "Waveform [1]"=::NewEnumerator13
     "[1] Frequency"=1
     "[1] Amplitude Min/Max"=<0 1>
     "[2] Average"=false
     "Waveform [2]"=::NewEnumerator9
     "[2] Input Position"=0
     "[2] Phase"=$Emitter.Age
     "[2] Phase Scale"=1
     "[2] Compound Frequency"=2
     "[2] Frequency"=2
     "[2] Amplitude Min/Max"=<0 1>
     "[2] Scale"=1
     "[2] Bias"=0
     "[2] Exponent"=1
     "[2] Pulse Duration"=0.1
     "[2] Clip Triangle Min/Max"=<0 0>
     "[2] Random Seed"=0
    )
    Module.Index0Param0Write=true
    Module.Index0Param1Write=true
    Module.Index0Param2Write=true
    Module.Index0Param3Write=true
    Module.Index1Param0Write=true
    Module.Index1Param1Write=true
    Module.Index1Param2Write=true
    Module.Index1Param3Write=true
    Write Parameter Index 1=true
   )
   &UpdateMeshOrientation(
    Facing Direction=&FN_GetSystemAxis(Axis:ENiagaraOrientationAxis=::YAxis, Flip Axis:NiagaraBool=false)
    Facing Interpolation=false
    Orientation Method=::NewEnumerator1
    Override Quat=false
    TransformFacing=false
    TransformUp=false
   )
   &SkeletalMeshLocation(
    Bone Sampling Mode=::NewEnumerator4
    MaskSpawnGroup=false
    Orientation Sampling=::NewEnumerator1
    Position Sampling=::NewEnumerator2
    Skeletal Mesh=$System.Skeletal Mesh
   ) {module=SkeletalMeshLocation002}
   &SkeletalMeshLocation(
    Bone / Socket Index=&SelectIntFromArray(Direct Array Index=$Particles.PartitionParticles001.Partition, Int Selection Array=DataInterface() {propertyValues {IntData [1,2]}})
    Bone Sampling Mode=::NewEnumerator4
    MaskSpawnGroup=false
    Orientation Sampling=::NewEnumerator1
    Position Sampling=::NewEnumerator2
    Skeletal Mesh=$System.Skeletal Mesh
   ) {module=SkeletalMeshLocation003}
   /SetVariables(
    Particles.BackerFlame Dyn 0:Vector4f=&MakeVector4(
     X=&Multiply_Float(
      A=&Multiply_Float(
       A=&FloatFromCurve(
        FloatCurve=DataInterface() {propertyValues {Curve {
           defaultValue=1
           keys [
            {interpMode=RCIM_Cubic, tangentMode=RCTM_User, time=0, value=0.85}
            {interpMode=RCIM_Cubic, tangentMode=RCTM_User, time=0.5, value=1}
            <at 1 0.85 RCIM_Cubic>
           ]
          }}}
        CurveIndex=$Particles.PartitionParticles001.NormalizedPartition
        Scale Curve=&Multiply_Float(A=25, B=$Particles.Waveform)
       )
       B=&MakeCustomFloatFromBool(Bool=$User.IsLargeBody, True Float=1.4, False Float=1)
      )
      B=1.15
     )
     Y=&RandomRangeFloat(Minimum=0.5, Maximum=1.2)
     Z=0
     W=0
    )
    Particles.BackerFlameColor:LinearColor=&MultiplyLinearColor(
     Linear Color=$Particles.Initial.BackerFlameColor
     Scale Factor=&MakeVector4FromVectorAndFloat(
      "Vector (RGB)"=&MakeVectorFromLinearColorRGB(Color=&ColorFromCurve(
        ColorCurve=$System.Color Curve
        CurveIndex=&Frac_Float(Float=&Add_Float(A=&Multiply_Float(A=$Engine.Time, B=0.3), B=&ReturnNormalizedExecIndex()))
       ))
      "Float (Alpha)"=&Multiply_Float(
       A=&Waveform(
        Global Phase=&Add_Float(A=&ReturnNormalizedExecIndex(), B=$Engine.Time)
        "Waveform [1]"=::NewEnumerator9
        "[1] Frequency"=5
        "[1] Amplitude Min/Max"=<0.4 1>
       )
       B=&FloatFromCurve(
        FloatCurve=DataInterface() {propertyValues {Curve {keys [
            <at 0.1 0.85 RCIM_Cubic>
            {arriveTangentWeight=0.1077, interpMode=RCIM_Cubic, leaveTangentWeight=0.0718, tangentMode=RCTM_User, time=0.4, value=1}
            {arriveTangentWeight=0.0718, interpMode=RCIM_Cubic, leaveTangentWeight=0.1077, tangentMode=RCTM_User, time=0.6, value=1}
            <at 0.9 0.85 RCIM_Cubic>
           ]}}}
        CurveIndex=Linked() {linkedVariable {name=NPC.FortniteNPC.TimeOfDay_0-1, type=NiagaraFloat}}
        Scale Curve=&MakeCustomFloatFromBool(Bool=$User.IsFrontEnd, True Float=0.8, False Float=1)
       )
      )
     )
    )
   ) {isSetParameters=true}
   &InterpolateOverTime(
    Rate Of Change=12
    Target Vector Value:Vector3f=&ConvertPositionToVector(Input Position=$Particles.Position, Convert To Absolute Large World Vector=false)
    Variable Type=::NewEnumerator2
   ) {enabled=false}
   /SetVariables(Particles.Interpolate Dynamic=&MakeFloat4FromVectorAndFloat("Vector (RGB)"=<1 1 1>, "Float (Alpha)"=0)) {isSetParameters=true, module=SetVariables_#6}
   &ScaleColor(
    Scale Alpha=&Multiply_Float(
     A=&Waveform(
      Global Phase=&Add_Float(A=&ReturnNormalizedExecIndex(), B=$Engine.Time)
      "Waveform [1]"=::NewEnumerator9
      "[1] Frequency"=5
      "[1] Amplitude Min/Max"=<0.4 1>
     )
     B=&FloatFromCurve(
      FloatCurve=DataInterface() {propertyValues {Curve {keys [
          <at 0.1 0.85 RCIM_Cubic>
          {arriveTangentWeight=0.1077, interpMode=RCIM_Cubic, leaveTangentWeight=0.0718, tangentMode=RCTM_User, time=0.4, value=1}
          {arriveTangentWeight=0.0718, interpMode=RCIM_Cubic, leaveTangentWeight=0.1077, tangentMode=RCTM_User, time=0.6, value=1}
          <at 0.9 0.85 RCIM_Cubic>
         ]}}}
      CurveIndex=Linked() {linkedVariable {name=NPC.FortniteNPC.TimeOfDay_0-1, type=NiagaraFloat}}
      Scale Curve=&MakeCustomFloatFromBool(Bool=$User.IsFrontEnd, True Float=0.8, False Float=1)
     )
    )
    Scale RGB=&MakeVectorFromLinearColorRGB(Color=&ColorFromCurve(
      ColorCurve=$System.Color Curve
      CurveIndex=&Frac_Float(Float=&Add_Float(A=&Multiply_Float(A=$Engine.Time, B=0.3), B=&ReturnNormalizedExecIndex()))
     ))
    ScaleA=true
    ScaleRGB=true
   ) {module=ScaleColor001}
   /SetVariables(Particles.Position=/Niagara/DynamicInputs/Math/LerpMultipleVectors(
     Alpha:NiagaraFloat=&ReturnNormalizedExecIndex()
     V 1:Vector3f=&AddVectorToPosition(Position=$Particles.SkeletalMeshLocation003.SampledPosition, Vector=<0 0 -1>)
     V 2:Vector3f=&AddVectorToPosition(Position=$Particles.SkeletalMeshLocation002.SampledPosition, Vector=<0 0 -2>)
     V 3:Vector3f=&AddVectorToPosition(Position=$Particles.SkeletalMeshLocation003.SampledPosition, Vector=<0 0 -1>)
     V 4:Vector3f=?
     V 5:Vector3f=?
     Vector Count:ENiagara_MultipleLerpCount=::NewEnumerator0
    )) {isSetParameters=true}
  ]
  rendererClasses [&NiagaraMeshRendererProperties]
  renderers [
   {rendererClass=&NiagaraMeshRendererProperties, rendererIndex=0}
   {rendererClass=&NiagaraMeshRendererProperties, rendererIndex=1}
   {rendererClass=&NiagaraMeshRendererProperties, rendererIndex=2}
  ]
 }
 Glow {
  emitterUpdateScript [
   &EmitterState(
    Inactive Response=::NewEnumerator1
    Life Cycle Mode=::NewEnumerator1
    Loop Behavior=::NewEnumerator1
    Loop Duration Mode=::NewEnumerator1
    UseLoopDelay=false
   )
   &SpawnBurst_Instantaneous(Spawn Count=7, Use Spawn Probability=false, UseLoopCountLimit=true)
  ]
  particleSpawnScript [
   &PartitionParticles(Number Of Partitions=7)
   &InitializeParticle(
    Lifetime=3
    Sprite Size Mode=::NewEnumerator3
    Uniform Sprite Size=&Multiply_Float(
     A=&SelectFloatFromArray(
      Direct Array Index=$Particles.PartitionParticles.Partition
      Float Selection Array=DataInterface() {propertyValues {FloatData [30,40,40,35,35,30,30]}}
     )
     B=&MakeCustomFloatFromBool(Bool=$User.IsLargeBody, True Float=1.3, False Float=1)
    )
    UsePositionOffset=false
    Write Material Random=true
   )
   &CameraOffset(Camera Offset Amount=&SelectFloatFromArray(
     Direct Array Index=$Particles.PartitionParticles.Partition
     Float Selection Array=DataInterface() {propertyValues {FloatData [6,10,10,10,10,10,10]}}
    ))
   /SetVariables(Particles.GlowCameraOffset=-30) {isSetParameters=true}
   /SetVariables(Particles.GlowScale:Vector2f=&Multiply_Vector2D(
     A=&SelectVector2DFromArray(
      Direct Array Index=$Particles.PartitionParticles.Partition
      Vector 2D Selection Array=DataInterface() {propertyValues {FloatData [<80 80>, <60 60>, <60 60>, <35 35>, <35 35>, <25 25>, <25 25>]}}
     )
     B=&MakeCustomVector2DFromBool(Bool=$User.IsLargeBody, True Vector2D=<1.4 1.4>, False Vector2D=<1 1>)
    )) {isSetParameters=true}
   /SetVariables(Particles.GlowColor=&Multiply_LinearColorByFloat(
     Linear Color=<rgba 1 1 1 1>
     Float=&SelectFloatFromArray(
      Direct Array Index=$Particles.PartitionParticles.Partition
      Float Selection Array=DataInterface() {propertyValues {FloatData [4,1.5,1.5,1,1,1,1]}}
     )
    )) {isSetParameters=true}
   /SetVariables(Particles.Decal Dynamic Param:Vector4f=&MakeVector4(X=5, Y=0.1, Z=0, W=0)) {isSetParameters=true}
  ]
  particleUpdateScript [
   &ParticleState(Kill Particles When Lifetime Has Elapsed=false, Let Infinitely Lived Particles Die When Emitter Deactivates=true)
   &SkeletalMeshLocation(
    Bone / Socket Index=&SelectIntFromArray(Direct Array Index=$Particles.PartitionParticles.Partition, Int Selection Array=DataInterface() {propertyValues {IntData [0,1,2,3,4,5,6]}})
    Bone Sampling Mode=::NewEnumerator4
    MaskSpawnGroup=false
    Orientation Sampling=::NewEnumerator1
    Sampled Position Offset=&SelectVectorFromArray(
     Direct Array Index=$Particles.PartitionParticles.Partition
     Vector Selection Array=DataInterface() {propertyValues {FloatData [<-5 0 0>, <0 0 0>, <0 0 0>, <0 0 0>, <0 0 0>, <0 0 0>, <0 0 0>]}}
    )
    Skeletal Mesh=$System.Skeletal Mesh
   )
   &SolveForcesAndVelocity(Clamp Velocity=false, Limit Acceleration=false)
   /SetVariables(
    Particles.GlowColor=&MakeLinearColorFromVectorAndFloat(
     "Vector (RGB)"=&Multiply_Vector(A=$System.Color RGB, B=&MakeVectorFromLinearColorRGB(Color=$Particles.Initial.GlowColor))
     "Float (Alpha)"=&Multiply_Float(
      A=&FloatFromCurve(
       FloatCurve=DataInterface() {propertyValues {Curve {keys [
           <at 0.1 0.8 RCIM_Cubic>
           {arriveTangentWeight=0.1077, interpMode=RCIM_Cubic, leaveTangentWeight=0.0718, tangentMode=RCTM_User, time=0.4, value=1}
           {arriveTangentWeight=0.0718, interpMode=RCIM_Cubic, leaveTangentWeight=0.1077, tangentMode=RCTM_User, time=0.6, value=1}
           <at 0.9 0.8 RCIM_Cubic>
          ]}}}
       CurveIndex=Linked() {linkedVariable {name=NPC.FortniteNPC.TimeOfDay_0-1, type=NiagaraFloat}}
      )
      B=&MakeCustomFloatFromBool(Bool=$User.IsFrontEnd, True Float=0.75, False Float=1)
     )
    )
    Particles.GlowScale:Vector2f=&Multiply_Vector2D(
     A=$Particles.Initial.GlowScale
     B=&Vector2DFromFloat(Value=&FloatFromCurve(FloatCurve=DataInterface() {propertyValues {Curve {keys [
           {interpMode=RCIM_Cubic, tangentMode=RCTM_User, time=0, value=0.8}
           {interpMode=RCIM_Cubic, tangentMode=RCTM_User, time=0.5, value=1}
           {tangentMode=RCTM_User, time=1, value=0.8}
          ]}}}))
    )
   ) {isSetParameters=true}
   &ScaleColor(Scale RGB=&Multiply_VectorByFloat(Vector=$System.Color RGB, Float=4), ScaleA=true, ScaleRGB=true)
  ]
  rendererClasses [&NiagaraSpriteRendererProperties]
  renderers [
   {rendererClass=&NiagaraSpriteRendererProperties}
   {rendererClass=&NiagaraSpriteRendererProperties, rendererIndex=1}
  ]
 }
 Lines {
  emitterUpdateScript [
   &EmitterState(Life Cycle Mode=::NewEnumerator1, Loop Behavior=::NewEnumerator1, Loop Duration Mode=::NewEnumerator1, UseLoopDelay=false)
   &SpawnBurst_Instantaneous(Use Spawn Probability=false, UseLoopCountLimit=false)
   &SpawnBurst_Instantaneous(Spawn Count=6, Spawn Group=1, Use Spawn Probability=false, UseLoopCountLimit=false) {module=SpawnBurst_Instantaneous001}
  ]
  particleSpawnScript [
   &PartitionParticles(
    Normalized Index In Partition Includes One=true
    Number Of Partitions=&MaskIntBySpawnGroup(Masked Int=1, Passthrough Int=2, Spawn Group Mask=0)
    Particles Per Partition:NiagaraInt32=&MaskIntBySpawnGroup(Masked Int=1, Passthrough Int=3, Spawn Group Mask=0)
    Partition Mode:ENiagara_PartitionMode_v2=::NewEnumerator1
   )
   &InitializeParticle(
    Color=&MaskLinearColorBySpawnGroup(
     Masked Linear Color=<rgba 1 1 1 0.43>
     Passthrough Linear Color=&LerpLinearColors(StartColor=<rgba 1 1 1 0.37>, EndColor=<rgba 1 1 1 0.2>, LerpFactor=$Particles.PartitionParticles.Normalized Index In Partition)
    )
    Lifetime=0.1
    Mesh Scale=&Multiply_VectorByFloat(
     Vector=&MaskVectorBySpawnGroup(
      Masked Vector=<1 1 1.6>
      Passthrough Vector=&Lerp_Vector(A=<1 1 1.3>, B=<1 1 1>, Alpha=$Particles.PartitionParticles.Normalized Index In Partition)
     )
     Float=&MakeCustomFloatFromBool(Bool=$User.IsLargeBody, True Float=1.2, False Float=1)
    )
    Mesh Scale Mode=::NewEnumerator4
    UsePositionOffset=false
    Write Material Random=true
   )
   &InitialMeshOrientation(Mesh Orientation Mode=::NewEnumerator1, Use Rotation Vector=true)
  ]
  particleUpdateScript [
   &ParticleState(Kill Particles When Lifetime Has Elapsed=false, Let Infinitely Lived Particles Die When Emitter Deactivates=true)
   &DynamicMaterialParameters(
    Index 0 Param 1=&Multiply_Float(
     A=&MaskFloatBySpawnGroup(
      Masked Float=50
      Passthrough Float=&Lerp_Float(A=30, B=24, Alpha=$Particles.PartitionParticles.Normalized Index In Partition)
     )
     B=&MakeCustomFloatFromBool(Bool=$User.IsLargeBody, True Float=1.25, False Float=1)
    )
    Module.Index0Float0Write=true
    Module.Index0Float1Write=true
    Module.Index0Float2Write=true
    Module.Index0Float3Write=true
   )
   &SkeletalMeshLocation(
    Bone / Socket Index=&MaskIntBySpawnGroup(
     Masked Int=7
     Passthrough Int=&SelectIntFromArray(Direct Array Index=$Particles.PartitionParticles.Partition, Int Selection Array=DataInterface() {propertyValues {IntData [1,2]}})
     Spawn Group Mask=0
    )
    Bone Sampling Mode=::NewEnumerator4
    MaskSpawnGroup=false
    Orientation Sampling=::NewEnumerator1
    Sampled Position Offset=&MaskVectorBySpawnGroup(
     Masked Vector=<7 0 0>
     Passthrough Vector=&Multiply_Vector(
      A=&Lerp_Vector(
       A=&SelectVectorFromArray(
        Direct Array Index=$Particles.PartitionParticles.Partition
        Vector Selection Array=DataInterface() {propertyValues {FloatData [<1 -4 -3>, <-1 4 3>]}}
       )
       B=&SelectVectorFromArray(
        Direct Array Index=$Particles.PartitionParticles.Partition
        Vector Selection Array=DataInterface() {propertyValues {FloatData [<24 2 -4>, <-24 -2 4>]}}
       )
       Alpha=$Particles.PartitionParticles.Normalized Index In Partition
      )
      B=&Multiply_VectorByFloat(
       Vector=&MakeCustomVectorFromBool(Bool=$User.IsLargeBody, True Vector=<1.2 1 0.5>, False Vector=<1 1 1>)
       Float=$System.EngineOwnerScale
      )
     )
    )
    Skeletal Mesh=$System.Skeletal Mesh
   )
   &UpdateMeshOrientation(
    Facing Direction=&FN_GetSystemAxis(Axis:ENiagaraOrientationAxis=::YAxis, Flip Axis:NiagaraBool=false)
    Facing Interpolation=false
    Orientation Method=::NewEnumerator1
    Override Quat=false
    TransformFacing=false
    TransformUp=false
   )
   &ScaleColor(
    Scale Alpha=&MaskFloatBySpawnGroup(
     Masked Float=&Multiply_Float(
      A=&Multiply_Float(
       A=&Waveform(
        Global Phase=&Add_Float(A=&ReturnNormalizedExecIndex(), B=$Engine.Time)
        "Waveform [1]"=::NewEnumerator9
        "[1] Frequency"=1
        "[1] Amplitude Min/Max"=<0.8 1>
       )
       B=&MakeCustomFloatFromBool(Bool=$User.IsFrontEnd, True Float=0.5, False Float=1)
      )
      B=&FloatFromCurve(
       FloatCurve=DataInterface() {propertyValues {Curve {keys [
           <at 0.1 0.85 RCIM_Cubic>
           {arriveTangentWeight=0.1077, interpMode=RCIM_Cubic, leaveTangentWeight=0.0718, tangentMode=RCTM_User, time=0.4, value=1}
           {arriveTangentWeight=0.0718, interpMode=RCIM_Cubic, leaveTangentWeight=0.1077, tangentMode=RCTM_User, time=0.6, value=1}
           <at 0.9 0.85 RCIM_Cubic>
          ]}}}
       CurveIndex=Linked() {linkedVariable {name=NPC.FortniteNPC.TimeOfDay_0-1, type=NiagaraFloat}}
      )
     )
     Passthrough Float=&Multiply_Float(
      A=&Multiply_Float(
       A=&Waveform(
        Global Phase=&Add_Float(A=&ReturnNormalizedExecIndex(), B=$Engine.Time)
        "Waveform [1]"=::NewEnumerator9
        "[1] Frequency"=1
        "[1] Amplitude Min/Max"=<0.8 1>
       )
       B=&MakeCustomFloatFromBool(Bool=$User.IsFrontEnd, True Float=0.5, False Float=1)
      )
      B=&FloatFromCurve(
       FloatCurve=DataInterface() {propertyValues {Curve {keys [
           <at 0.1 0.85 RCIM_Cubic>
           {arriveTangentWeight=0.1077, interpMode=RCIM_Cubic, leaveTangentWeight=0.0718, tangentMode=RCTM_User, time=0.4, value=1}
           {arriveTangentWeight=0.0718, interpMode=RCIM_Cubic, leaveTangentWeight=0.1077, tangentMode=RCTM_User, time=0.6, value=1}
           <at 0.9 0.85 RCIM_Cubic>
          ]}}}
       CurveIndex=Linked() {linkedVariable {name=NPC.FortniteNPC.TimeOfDay_0-1, type=NiagaraFloat}}
      )
     )
    )
    Scale RGB=&MakeVectorFromLinearColorRGB(Color=<rgba 1 1 1 1>)
    ScaleA=true
    ScaleRGB=false
   ) {module=ScaleColor001}
   &InterpolateOverTime(
    Rate Of Change=12
    Target Vector Value:Vector3f=&ConvertPositionToVector(Input Position=$Particles.Position, Convert To Absolute Large World Vector=false)
    Variable Type=::NewEnumerator2
   ) {enabled=false}
   /SetVariables(Particles.Interpolate Dynamic=&MakeFloat4FromVectorAndFloat("Vector (RGB)"=<1 1 1>, "Float (Alpha)"=0)) {isSetParameters=true, module=SetVariables_#8}
  ]
  rendererClasses [&NiagaraMeshRendererProperties]
  renderers [{rendererClass=&NiagaraMeshRendererProperties, rendererIndex=0}]
 }
 Lines_Arms {
  bEnabled=false
  emitterUpdateScript [
   &EmitterState(Life Cycle Mode=::NewEnumerator1, Loop Behavior=::NewEnumerator1, Loop Duration Mode=::NewEnumerator1, UseLoopDelay=false)
   &SpawnBurst_Instantaneous(Spawn Count=6, Use Spawn Probability=false, UseLoopCountLimit=false)
  ]
  particleSpawnScript [
   &PartitionParticles(Normalized Index In Partition Includes One=true, Particles Per Partition=3, Partition Mode:ENiagara_PartitionMode_v2=::NewEnumerator1)
   &InitializeParticle(
    Color=&LerpLinearColors(StartColor=<rgba 1 1 1 0.4>, EndColor=<rgba 1 1 1 0.2>, LerpFactor=$Particles.PartitionParticles.Normalized Index In Partition)
    Lifetime=0.1
    Mesh Scale=&Multiply_VectorByFloat(
     Vector=&Lerp_Vector(A=<1 1 1.3>, B=<1 1 1>, Alpha=$Particles.PartitionParticles.Normalized Index In Partition)
     Float=&MakeCustomFloatFromBool(Bool=$User.IsLargeBody, True Float=1.2, False Float=1)
    )
    Mesh Scale Mode=::NewEnumerator4
    UsePositionOffset=false
    Write Material Random=true
   )
   &InitialMeshOrientation(Mesh Orientation Mode=::NewEnumerator1, Use Rotation Vector=true)
  ]
  particleUpdateScript [
   &ParticleState(Kill Particles When Lifetime Has Elapsed=false, Let Infinitely Lived Particles Die When Emitter Deactivates=true)
   &DynamicMaterialParameters(
    Index 0 Param 1=&Multiply_Float(
     A=&Lerp_Float(A=30, B=24, Alpha=$Particles.PartitionParticles.Normalized Index In Partition)
     B=&MakeCustomFloatFromBool(Bool=$User.IsLargeBody, True Float=1.2, False Float=1)
    )
    Module.Index0Float0Write=true
    Module.Index0Float1Write=true
    Module.Index0Float2Write=true
    Module.Index0Float3Write=true
   )
   &SkeletalMeshLocation(
    Bone / Socket Index=&SelectIntFromArray(Direct Array Index=$Particles.PartitionParticles.Partition, Int Selection Array=DataInterface() {propertyValues {IntData [1,2]}})
    Bone Sampling Mode=::NewEnumerator4
    MaskSpawnGroup=false
    Orientation Sampling=::NewEnumerator1
    Sampled Position Offset=&Multiply_Vector(
     A=&Lerp_Vector(
      A=&SelectVectorFromArray(
       Direct Array Index=$Particles.PartitionParticles.Partition
       Vector Selection Array=DataInterface() {propertyValues {FloatData [<1 -4 -3>, <-1 4 3>]}}
      )
      B=&SelectVectorFromArray(
       Direct Array Index=$Particles.PartitionParticles.Partition
       Vector Selection Array=DataInterface() {propertyValues {FloatData [<24 2 -4>, <-24 -2 4>]}}
      )
      Alpha=$Particles.PartitionParticles.Normalized Index In Partition
     )
     B=&Multiply_VectorByFloat(
      Vector=&MakeCustomVectorFromBool(Bool=$User.IsLargeBody, True Vector=<1.2 1 0.5>, False Vector=<1 1 1>)
      Float=$System.EngineOwnerScale
     )
    )
    Skeletal Mesh=$System.Skeletal Mesh
   ) {module=SkeletalMeshLocation001}
   &UpdateMeshOrientation(
    Facing Direction=&FN_GetSystemAxis(Axis:ENiagaraOrientationAxis=::YAxis, Flip Axis:NiagaraBool=false)
    Facing Interpolation=false
    Orientation Method=::NewEnumerator1
    Override Quat=false
    TransformFacing=false
    TransformUp=false
   )
   &SkeletalMeshLocation(
    Bone / Socket Index=7
    Bone Sampling Mode=::NewEnumerator4
    MaskSpawnGroup=false
    Orientation Sampling=::NewEnumerator1
    Sampled Position Offset=&Multiply_Vector(
     A=&MakeVector(
      X=&FloatFromCurve(
       FloatCurve=DataInterface() {propertyValues {Curve {
          defaultValue=1
          keys [
           {interpMode=RCIM_Cubic, tangentMode=RCTM_User, time=0, value=18}
           {interpMode=RCIM_Cubic, tangentMode=RCTM_User, time=0.5, value=25}
           {tangentMode=RCTM_User, time=1, value=18}
          ]
         }}}
       CurveIndex=&ReturnNormalizedExecIndex()
       Scale Curve=1.5
      )
      Y=&FloatFromCurve(
       FloatCurve=DataInterface() {propertyValues {Curve {keys [
           {interpMode=RCIM_Cubic, leaveTangentWeight=0.1667, tangentMode=RCTM_User, time=0, value=2}
           {arriveTangentWeight=0.1667, interpMode=RCIM_Cubic, leaveTangentWeight=0.1667, tangentMode=RCTM_User, time=0.5, value=-2}
           {arriveTangentWeight=0.1667, interpMode=RCIM_Cubic, tangentMode=RCTM_User, time=1, value=2}
          ]}}}
       CurveIndex=&ReturnNormalizedExecIndex()
      )
      Z=&FloatFromCurve(
       FloatCurve=DataInterface() {propertyValues {Curve {keys [<at 0 15>,<at 1 -15>]}}}
       CurveIndex=&ReturnNormalizedExecIndex()
      )
     )
     B=&Multiply_VectorByFloat(
      Vector=&MakeCustomVectorFromBool(Bool=$User.IsLargeBody, True Vector=<1.1 1 1.3>, False Vector=<1 1 1>)
      Float=$System.EngineOwnerScale
     )
    )
    Skeletal Mesh=$System.Skeletal Mesh
   ) {enabled=false}
   &ScaleColor(
    Scale Alpha=&Multiply_Float(
     A=&Multiply_Float(
      A=&Waveform(
       Global Phase=&Add_Float(A=&ReturnNormalizedExecIndex(), B=$Engine.Time)
       "Waveform [1]"=::NewEnumerator9
       "[1] Frequency"=1
       "[1] Amplitude Min/Max"=<0.8 1>
      )
      B=&MakeCustomFloatFromBool(Bool=$User.IsFrontEnd, True Float=0.5, False Float=1)
     )
     B=&FloatFromCurve(
      FloatCurve=DataInterface() {propertyValues {Curve {keys [
          <at 0.1 0.85 RCIM_Cubic>
          {arriveTangentWeight=0.1077, interpMode=RCIM_Cubic, leaveTangentWeight=0.0718, tangentMode=RCTM_User, time=0.4, value=1}
          {arriveTangentWeight=0.0718, interpMode=RCIM_Cubic, leaveTangentWeight=0.1077, tangentMode=RCTM_User, time=0.6, value=1}
          <at 0.9 0.85 RCIM_Cubic>
         ]}}}
      CurveIndex=Linked() {linkedVariable {name=NPC.FortniteNPC.TimeOfDay_0-1, type=NiagaraFloat}}
     )
    )
    Scale RGB=&MakeVectorFromLinearColorRGB(Color=<rgba 1 1 1 1>)
    ScaleA=true
    ScaleRGB=false
   ) {module=ScaleColor001}
   &InterpolateOverTime(
    Rate Of Change=12
    Target Vector Value:Vector3f=&ConvertPositionToVector(Input Position=$Particles.Position, Convert To Absolute Large World Vector=false)
    Variable Type=::NewEnumerator2
   ) {enabled=false}
   /SetVariables(Particles.Interpolate Dynamic=&MakeFloat4FromVectorAndFloat("Vector (RGB)"=<1 1 1>, "Float (Alpha)"=0)) {isSetParameters=true}
  ]
  rendererClasses [&NiagaraMeshRendererProperties]
  renderers [{rendererClass=&NiagaraMeshRendererProperties, rendererIndex=0}]
 }
 Lines_Back {
  bEnabled=false
  emitterUpdateScript [
   &EmitterState(Life Cycle Mode=::NewEnumerator1, Loop Behavior=::NewEnumerator1, Loop Duration Mode=::NewEnumerator1, UseLoopDelay=false)
   &SpawnBurst_Instantaneous(Use Spawn Probability=false, UseLoopCountLimit=false)
  ]
  particleSpawnScript [
   &InitializeParticle(
    Color=<rgba 1 1 1 0.45>
    Lifetime=0.1
    Mesh Scale=&Multiply_VectorByFloat(
     Vector=<1 1 1.6>
     Float=&MakeCustomFloatFromBool(Bool=$User.IsLargeBody, True Float=1.2, False Float=1)
    )
    Mesh Scale Mode=::NewEnumerator4
    UsePositionOffset=false
    Write Material Random=true
   )
   &InitialMeshOrientation(Mesh Orientation Mode=::NewEnumerator1, Use Rotation Vector=true)
  ]
  particleUpdateScript [
   &ParticleState(Kill Particles When Lifetime Has Elapsed=false, Let Infinitely Lived Particles Die When Emitter Deactivates=true)
   &DynamicMaterialParameters(
    Index 0 Param 1=&Multiply_Float(A=50, B=&MakeCustomFloatFromBool(Bool=$User.IsLargeBody, True Float=1.25, False Float=1))
    Module.Index0Float0Write=true
    Module.Index0Float1Write=true
    Module.Index0Float2Write=true
    Module.Index0Float3Write=true
   )
   &SkeletalMeshLocation(
    Bone / Socket Index=7
    Bone Sampling Mode=::NewEnumerator4
    MaskSpawnGroup=false
    Orientation Sampling=::NewEnumerator1
    Sampled Position Offset=<7 0 0>
    Skeletal Mesh=$System.Skeletal Mesh
   )
   &UpdateMeshOrientation(
    Facing Direction=&FN_GetSystemAxis(Axis:ENiagaraOrientationAxis=::YAxis, Flip Axis:NiagaraBool=false)
    Facing Interpolation=false
    Orientation Method=::NewEnumerator1
    Override Quat=false
    TransformFacing=false
    TransformUp=false
   )
   &ScaleColor(
    Scale Alpha=&Multiply_Float(
     A=&Multiply_Float(
      A=&Waveform(
       Global Phase=&Add_Float(A=&ReturnNormalizedExecIndex(), B=$Engine.Time)
       "Waveform [1]"=::NewEnumerator9
       "[1] Frequency"=1
       "[1] Amplitude Min/Max"=<0.8 1>
      )
      B=&MakeCustomFloatFromBool(Bool=$User.IsFrontEnd, True Float=0.5, False Float=1)
     )
     B=&FloatFromCurve(
      FloatCurve=DataInterface() {propertyValues {Curve {keys [
          <at 0.1 0.85 RCIM_Cubic>
          {arriveTangentWeight=0.1077, interpMode=RCIM_Cubic, leaveTangentWeight=0.0718, tangentMode=RCTM_User, time=0.4, value=1}
          {arriveTangentWeight=0.0718, interpMode=RCIM_Cubic, leaveTangentWeight=0.1077, tangentMode=RCTM_User, time=0.6, value=1}
          <at 0.9 0.85 RCIM_Cubic>
         ]}}}
      CurveIndex=Linked() {linkedVariable {name=NPC.FortniteNPC.TimeOfDay_0-1, type=NiagaraFloat}}
     )
    )
    Scale RGB=&MakeVectorFromLinearColorRGB(Color=<rgba 1 1 1 1>)
    ScaleA=true
    ScaleRGB=false
   ) {module=ScaleColor001}
   &InterpolateOverTime(
    Rate Of Change=12
    Target Vector Value:Vector3f=&ConvertPositionToVector(Input Position=$Particles.Position, Convert To Absolute Large World Vector=false)
    Variable Type=::NewEnumerator2
   ) {enabled=false}
   /SetVariables(Particles.Interpolate Dynamic=&MakeFloat4FromVectorAndFloat("Vector (RGB)"=<1 1 1>, "Float (Alpha)"=0)) {isSetParameters=true, module=SetVariables_#8}
  ]
  rendererClasses [&NiagaraMeshRendererProperties]
  renderers [{rendererClass=&NiagaraMeshRendererProperties, rendererIndex=0}]
 }
 Trail_Sparkles {
  emitterUpdateScript [
   &EmitterState()
   &SpawnRate(Spawn Probability=0.75, SpawnRate=35, Use Spawn Probability=true)
  ]
  particleSpawnScript [
   &FN_SetSpawnGroup(Number of Spawn Groups=5)
   &ShapeLocation(MaskSpawnGroup=false, Non Uniform Scale=<1 1 0.5>, Override Seed=false, Sphere Radius=10, UseOwnerScale=true)
   &InitializeParticle(
    Color=&RandomRangeLinearColor(Minimum=<rgba 0.7803 0.5259 1 1>, Maximum=<rgba 0.4601 0.4832 1 1>)
    Lifetime Max=1
    Lifetime Min=0.5
    Lifetime Mode=::NewEnumerator1
    Sprite Rotation Mode=::NewEnumerator0
    Sprite Size Mode=::NewEnumerator1
    Uniform Sprite Size Max=4
    Uniform Sprite Size Min=3
    UsePositionOffset=false
    Write Material Random=true
   )
   &AddVelocity(
    Cone Angle=10
    Cone Axis=<0 0 1>
    Module.UseConeFalloff=false
    Module.UseVelDistribution=false
    Override Seed=false
    Rotation Coordinate Space=::World
    Velocity Mode=::NewEnumerator2
    Velocity Speed=&Multiply_Float(
     A=&RandomRangeFloat(Minimum=175, Maximum=350)
     B=&MakeCustomFloatFromBool(Bool=$User.IsLargeBody, True Float=1.3, False Float=1)
    )
   )
   /SetVariables(Particles.RandomCore:NiagaraFloat=&FloatFromCurve(
     FloatCurve=DataInterface() {propertyValues {Curve {keys [
         {arriveTangent=0.0559, interpMode=RCIM_Cubic, leaveTangent=0.0559, tangentMode=RCTM_User, time=0, value=0.2}
         {arriveTangent=0.1895, interpMode=RCIM_Cubic, leaveTangent=0.1895, tangentMode=RCTM_User, time=0.9, value=0.25}
         {arriveTangent=1.1808, interpMode=RCIM_Cubic, leaveTangent=1.1808, tangentMode=RCTM_User, time=1, value=0.3}
        ]}}}
     CurveIndex=&RandomRangeFloat(Minimum=0, Maximum=1)
     Scale Curve=2
    )) {isSetParameters=true}
   &SkeletalMeshLocation(
    Bone / Socket Index=$Particles.SpawnGroup
    Bone Sampling Mode=::NewEnumerator4
    "Inherit Velocity (Scale)"=-0.1
    Inherited Velocity Speed Limit=75
    MaskSpawnGroup=false
    Orientation Sampling=::NewEnumerator1
    Position Sampling=::NewEnumerator1
    Sampled Position Offset=$Particles.Initial.ShapeLocation.ShapePosition
    Skeletal Mesh=$System.Skeletal Mesh
    Velocity Sampling=::NewEnumerator0
    VelocitySpeedLimit=true
   )
  ]
  particleUpdateScript [
   /SetVariables(Particles.FlickerWaveform:NiagaraFloat=&Waveform(
     Global Phase=&Add_Float(A=$Engine.Time, B=&RandomRangeFloat(Minimum=0, Maximum=1))
     "Waveform [1]"=::NewEnumerator10
     "[1] Frequency"=20
     "[1] Amplitude Min/Max"=<0.2 1.2>
    )) {isSetParameters=true}
   &ParticleState()
   &CurlNoiseForce(Noise Strength=&Multiply_Float(A=300, B=1), Pan Noise=true, Pan Noise Field=<0 0 1>) {enabled=false}
   &ScaleSpriteSize(Scale Factor=&Vector2DFromFloat(Value=&FloatFromCurve(
      FloatCurve=DataInterface() {propertyValues {Curve {keys [
          {arriveTangent=7.4528, interpMode=RCIM_Cubic, leaveTangent=7.4528, tangentMode=RCTM_User, time=0, value=0}
          <at 0.2 1.2 RCIM_Cubic>
          {arriveTangent=-2.7132, interpMode=RCIM_Cubic, leaveTangent=-2.7132, tangentMode=RCTM_User, time=1, value=0}
         ]}}}
      Scale Curve=&Multiply_Float(A=&FN_FrontEndFloatModifier(Gameplay=0.4, Front End=0.3), B=$Particles.FlickerWaveform)
     )))
   &ScaleColor(
    Scale RGB=&Multiply_VectorByFloat(
     Vector=&MakeVectorFromLinearColorRGB(Color=&ColorFromCurve(
       ColorCurve=$System.Color Curve
       CurveIndex=&Frac_Float(Float=&Add_Float(A=&RandomRangeFloat(Minimum=0, Maximum=1), B=$Particles.NormalizedAge))
      ))
     Float=&FloatFromCurve(
      FloatCurve=DataInterface() {propertyValues {Curve {
         defaultValue=1
         keys [
          {interpMode=RCIM_Cubic, tangentMode=RCTM_User, time=0, value=1}
          {arriveTangent=-2.2239, interpMode=RCIM_Cubic, leaveTangent=-2.2239, tangentMode=RCTM_User, time=1, value=0.2}
         ]
        }}}
      Scale Curve=&Multiply_Float(
       A=4
       B=/Game/Effects/Niagara/DynamicInputs/Math/FN_Fit(
        Clamp:NiagaraBool=false
        Value:NiagaraFloat=$Particles.FlickerWaveform
        SourceMin:NiagaraFloat=0.2
        SourceMax:NiagaraFloat=1.2
        DestMin:NiagaraFloat=0.75
        DestMax:NiagaraFloat=1.2
       )
      )
     )
    )
    ScaleA=true
    ScaleRGB=true
   )
   /SetVariables(Particles.Glow03_Size=&Multiply_Vector2D_ByFloat(
     Float=&FloatFromCurve(
      FloatCurve=DataInterface() {propertyValues {Curve {keys [
          <at 0 0 RCIM_Cubic>
          <at 0.2 1.3 RCIM_Cubic>
          {arriveTangent=-0.1423, interpMode=RCIM_Cubic, leaveTangent=-0.1423, tangentMode=RCTM_User, time=1, value=0}
         ]}}}
      Scale Curve=&Multiply_Float(A=&MakeFloatFromVector2d(Vector 2D=$Particles.SpriteSize), B=3.5)
     )
     Vector2D=<1 1.5>
    )) {isSetParameters=true}
   /SetVariables(Particles.Glow03_Color=&MakeLinearColorFromVectorAndFloat(
     "Vector (RGB)"=&MakeVectorFromLinearColorRGB(Color=&MultiplyLinearColor(Linear Color=$Particles.Color, Scale Factor=&MakeVector4_FromLinearColor(LinearColor=$Particles.Color)))
     "Float (Alpha)"=&Multiply_Float(A=0.03, B=1)
    )) {isSetParameters=true}
   &Drag(Drag=&RandomRangeFloat(Minimum=8, Maximum=12), Use Linear Drag=true, Use Rotational Drag=false)
   &SolveForcesAndVelocity(Clamp Velocity=false, Limit Acceleration=false)
   &CameraOffset(Camera Offset Amount=-10)
  ]
  rendererClasses [&NiagaraSpriteRendererProperties]
  renderers [
   {rendererClass=&NiagaraSpriteRendererProperties}
   {rendererClass=&NiagaraSpriteRendererProperties, rendererIndex=1}
  ]
 }
}
summary {
 systemName=NS_Aura_BrightFlame
 emitters [
  {emitterName=Glow, rendererClasses [&NiagaraSpriteRendererProperties]}
  {emitterName=Trail_Sparkles, rendererClasses [&NiagaraSpriteRendererProperties]}
  {emitterName=Dust, rendererClasses [&NiagaraSpriteRendererProperties]}
  {bEnabled=false, emitterName=Lines_Back, rendererClasses [&NiagaraMeshRendererProperties]}
  {bEnabled=false, emitterName=Lines_Arms, rendererClasses [&NiagaraMeshRendererProperties]}
  {bEnabled=false, emitterName=Flames_Back_Original, rendererClasses [&NiagaraMeshRendererProperties]}
  {emitterName=Flames_Arms, rendererClasses [&NiagaraMeshRendererProperties]}
  {emitterName=EnergyWisps, rendererClasses [&NiagaraSpriteRendererProperties]}
  {emitterName=Lines, rendererClasses [&NiagaraMeshRendererProperties]}
  {emitterName=Flames_Back_Revised, rendererClasses [&NiagaraMeshRendererProperties]}
 ]
 userVariables [
  User.bIsFrontEnd {type=NiagaraBool}
  User.bIsFrontEndPreview {type=NiagaraBool}
  User.IsFrontEnd {type=NiagaraBool}
  User.IsLargeBody {type=NiagaraBool}
 ]
}
